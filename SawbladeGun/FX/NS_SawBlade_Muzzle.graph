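# Niagara muzzle effect for the sawblade gun, built from four burst emitters.
# A GPU mesh flash and a brief LightFlash with light, sprite and mesh renderers form the flash.
# CPU smoke sprites inherit velocity and erode through curve-driven material parameters.
# GPU sparks use curl noise, collision and time-of-day alpha, with one skewed random value linking lifetime, gravity and speed-stretch threshold.
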NiagaraSystem /SawbladeGun/FX/NS_SawBlade_Muzzle
dependencies {
 usedDataInterfaces [
  &NiagaraDataInterfaceVectorCurve
  &NiagaraDataInterfaceCurve
  &NiagaraDataInterfaceCamera
  &NiagaraDataInterfaceColorCurve
  &NiagaraDataInterfaceVector2DCurve
 ]
 usedDynamicInputs [&MakeVector4_FromLinearColor, &ColorFromCurve]
 usedModules [&SystemState]
}
emitters {
 FN_Weapons_Smoke_BETA001 {
  emitterUpdateScript [
   &EmitterState()
   &SpawnBurst_Instantaneous(Spawn Count=&UniformRangedInt(Minimum=2, Maximum=5), Use Spawn Probability=false)
  ]
  particleSpawnScript [
   &InitializeParticle(
    Lifetime Max=0.5
    Lifetime Min=0.2
    Lifetime Mode=::NewEnumerator1
    Sprite Rotation Angle=&RandomRangeFloat(Minimum=0, Maximum=1)
    Sprite Rotation Mode=::NewEnumerator1
    Sprite Size Mode=::NewEnumerator1
    Uniform Sprite Size Max=100
    Uniform Sprite Size Min=60
    UsePositionOffset=false
    Write Material Random=true
   ) {module=InitializeParticle001}
   /SetVariables(Particles.InitSpriteRotRate=&UniformRangedFloat(Minimum=-3, Maximum=3)) {isSetParameters=true}
   &AddVelocityInCone(
    Cone Angle=10
    Module.Use Velocity Falloff On Cone Axis=false
    Velocity Strength=&UniformRangedFloat(Minimum=150, Maximum=700)
   )
   &InheritVelocity(Inherited Velocity Amount Scale=<0.5 0.5 0.5>, Inherited Velocity Speed Limit=2000, Speed Limit=true, UseSpeedThreshold=false)
   &CameraOffset(Camera Offset Amount=14)
  ]
  particleUpdateScript [
   &ParticleState()
   &SpriteRotationRate(Rotation Rate=&Multiply_Float(
     A=$Particles.InitSpriteRotRate
     B=&FloatFromCurve(FloatCurve=DataInterface() {propertyValues {Curve {keys [<at -0.0046 1.0113 RCIM_Cubic>,<at 1.0011 0.09 RCIM_Cubic>]}}})
    ))
   &Drag(Use Linear Drag=true, Use Rotational Drag=false)
   &SolveForcesAndVelocity(Clamp Velocity=false, Limit Acceleration=false)
   &ScaleSpriteSize(Scale Factor=&Vector2DFromFloat(Value=&FloatFromCurve(
      FloatCurve=DataInterface() {propertyValues {Curve {keys [
          {arriveTangent=2.8364, interpMode=RCIM_Cubic, leaveTangent=2.8364, tangentMode=RCTM_User, time=0.0013, value=0.1}
          {arriveTangent=0.8258, interpMode=RCIM_Cubic, leaveTangent=0.8258, tangentMode=RCTM_User, time=0.5367, value=0.8831}
          {arriveTangent=-0.0035, interpMode=RCIM_Cubic, leaveTangent=-0.0035, tangentMode=RCTM_User, time=1.0008, value=1.6042}
         ]}}}
      Scale Curve=1.25
     )))
   /Game/Effects/Niagara/Modules/Materials/MaterialSmoke(
    Erode:NiagaraFloat=&FloatFromCurve(FloatCurve=DataInterface() {propertyValues {
       bOptimizeLUT=false
       bUseLUT=false
       Curve {keys [
         {arriveTangent=-0.0652, interpMode=RCIM_Cubic, leaveTangent=-0.0652, tangentMode=RCTM_User, time=0.11, value=0.97}
         {arriveTangent=-5.7797, interpMode=RCIM_Cubic, leaveTangent=-5.7797, tangentMode=RCTM_User, time=0.39, value=0.66}
         {arriveTangent=-0.342, interpMode=RCIM_Cubic, leaveTangent=-0.342, tangentMode=RCTM_User, time=0.9954, value=-0.74}
        ]}
      }})
    Shading:NiagaraFloat=&FloatFromCurve(
     FloatCurve=DataInterface() {propertyValues {
       bOptimizeLUT=false
       bUseLUT=false
       Curve {keys [
         <at 0.07 0.5 RCIM_Cubic>
         {arriveTangent=1.2205, interpMode=RCIM_Cubic, leaveTangent=1.2205, time=0.17, value=1.02}
         {interpMode=RCIM_Cubic, tangentMode=RCTM_User, time=0.3895, value=0.89}
         <at 0.8 0 RCIM_Cubic>
        ]}
      }}
     Scale Curve=2
    )
   )
   &DynamicMaterialParameters(
    Index 0 Param 1=&Add_Float(
     A=$Particles.RandomFloat
     B=&FloatFromCurve(FloatCurve=DataInterface() {propertyValues {Curve {keys [<at 0 1>,<at 1 0>]}}})
    )
    Index 0 Param 2=&FloatFromCurve(FloatCurve=DataInterface() {propertyValues {Curve {keys [
         {arriveTangent=0.8564, interpMode=RCIM_Cubic, leaveTangent=0.8564, tangentMode=RCTM_User, time=0.0046, value=-0.9753}
         {arriveTangent=4.2876, interpMode=RCIM_Cubic, leaveTangent=5.5268, tangentMode=RCTM_User, time=1, value=1}
        ]}}})
    Index 0 Param 3=&FloatFromCurve(FloatCurve=DataInterface() {propertyValues {Curve {keys [
         {arriveTangent=-22.7009, interpMode=RCIM_Cubic, leaveTangent=-9.0491, tangentMode=RCTM_Break, time=0, value=1}
         {arriveTangent=-0.541, interpMode=RCIM_Cubic, leaveTangent=-28.3175, tangentMode=RCTM_Break, time=0.3531, value=0}
        ]}}})
    Index 0 Param 4=&FloatFromCurve(FloatCurve=DataInterface() {propertyValues {Curve {keys [{arriveTangent=-3.2497, interpMode=RCIM_Cubic, leaveTangent=-3.2497, tangentMode=RCTM_User, time=0, value=1}, <at 1 0 RCIM_Cubic>]}}})
    Module.Param0WriteEnabled=true
    Module.Param1WriteEnabled=true
    Module.Param2WriteEnabled=true
    Module.Param3WriteEnabled=true
   )
   &ScaleColor(
    Scale Alpha=&FloatFromCurve(FloatCurve=DataInterface() {propertyValues {Curve {keys [<at -0.003 -0.006>,<at 0.1806 0.5257>,<at 0.6 0.4>,<at 1 0>]}}}, Scale Curve=0.2)
    Scale RGB=/Niagara/DynamicInputs/ScaleByCurve/ScaleVectorByCurve(
     InitialValue:Vector3f=<1 1 1>
     ScaleCurve:NiagaraDataInterfaceVectorCurve=DataInterface() {propertyValues {XCurve {keys [<at 0.1 2>,<at 0.6 0.1>]}, YCurve {keys [<at 0.1 1.5>,<at 0.6 0.1>]}, ZCurve {keys [<at 0.1 1>,<at 0.6 0.1>]}}}
     CurveIndex:NiagaraFloat=$Particles.NormalizedAge
    )
    ScaleA=true
    ScaleRGB=true
   )
  ]
  rendererClasses [&NiagaraSpriteRendererProperties]
  renderers [
   {rendererClass=&NiagaraSpriteRendererProperties}
   {rendererClass=&NiagaraSpriteRendererProperties, rendererIndex=1}
  ]
 }
 LightFlash {
  emitterUpdateScript [
   &EmitterScalability(
    Module.bApplySpawnCountScale=false
    Module.bApplySpawnCountScaleByDistanceFraction=false
    Module.bUseMaxDistance=false
    Module.bUseMinDistance=false
   ) {enabled=false}
   &EmitterLifeCycle(NextLoopDuration=4.4208) {enabled=false}
   &EmitterState()
   &SpawnBurst_Instantaneous(Spawn Count=1, Use Spawn Probability=false)
  ]
  particleSpawnScript [
   &InitializeParticle(
    Color=<rgba 15 5.3079 1.2136 1>
    Lifetime=0.1
    Mesh Scale Mode=::NewEnumerator3
    Position Offset=<10 0 0>
    Position Offset Coordinate Space=::Local
    Sprite Size Mode=::NewEnumerator3
    Uniform Sprite Size=200
    UsePositionOffset=true
    Write Material Random=true
   ) {module=InitializeParticle001}
   /Niagara/Modules/Spawn/Initialization/InitializeParticle(
    Color=<rgba 5 1.4433 1.2442 1>
    Lifetime=0.2
    Mesh Scale=&VectorFromFloat(Value=2)
    Module.Write Color=true
    Module.Write Lifetime=true
    Module.Write Mass=false
    Module.Write Material Random=true
    Module.Write Position=true
    Module.Write Scale=true
    Module.Write SpriteRotation=false
    Module.Write SpriteSize=true
    Sprite Size=&Vector2DFromFloat(Value=200)
   ) {enabled=false}
   /SetVariables(Particles.DynamicParams2:Vector4f=<1 3 0 0>) {isSetParameters=true}
   &CameraOffset(Camera Offset Amount=7)
  ]
  particleUpdateScript [
   &ParticleState()
   &Color(Color=<rgba 5000 4092.448 1796.875 1>) {enabled=false}
   &DynamicMaterialParameters(
    Index 0 Param 1=&FloatFromCurve(
     FloatCurve=DataInterface() {propertyValues {Curve {keys [{arriveTangent=2.0031, interpMode=RCIM_Cubic, time=0, value=0}, <at 1 1>]}}}
     Scale Curve=2
    )
    Index 0 Param 3=-1
    Index 0 Param 4=&FloatFromCurve(
     FloatCurve=DataInterface() {propertyValues {Curve {keys [
         {arriveTangent=9.8239, interpMode=RCIM_Cubic, leaveTangent=17.2663, tangentMode=RCTM_User, time=0, value=1.5}
         {arriveTangent=13.6522, interpMode=RCIM_Cubic, leaveTangent=13.6522, tangentMode=RCTM_User, time=0.3, value=-1}
         {arriveTangent=-6.5305, interpMode=RCIM_Cubic, leaveTangent=-6.5305, tangentMode=RCTM_User, time=0.5661, value=0.1474}
         <at 0.9 0>
        ]}}}
     Scale Curve=0.75
    )
    Index 1 Param 1=25
    Module.Param0WriteEnabled=true
    Module.Param0WriteEnabled1=true
    Module.Param1WriteEnabled=false
    Module.Param1WriteEnabled1=false
    Module.Param2WriteEnabled=false
    Module.Param2WriteEnabled1=false
    Module.Param3WriteEnabled=true
    Module.Param3WriteEnabled1=false
    Write Parameter Index 1=true
   )
   &ScaleColor(
    Scale Alpha=&FloatFromCurve(FloatCurve=DataInterface() {propertyValues {Curve {keys [
         {arriveTangent=-1.9974, interpMode=RCIM_Cubic, leaveTangent=-1.9974, tangentMode=RCTM_User, time=0, value=0.9643}
         {interpMode=RCIM_Cubic, tangentMode=RCTM_User, time=1, value=0}
        ]}}})
    Scale RGB=&VectorFromFloat(Value=&FloatFromCurve(FloatCurve=DataInterface() {propertyValues {Curve {keys [{arriveTangent=-2.5941, interpMode=RCIM_Cubic, time=0, value=1}, <at 1 0.5>]}}}))
    ScaleA=true
    ScaleRGB=true
   )
   /SetVariables(Particles.LightColor=&MultiplyLinearColor(
     Linear Color=$Particles.Color
     Scale Factor=&MakeVector4_FromFloat(Float=&FloatFromCurve(
       FloatCurve=DataInterface() {propertyValues {Curve {keys [{arriveTangent=-2.8123, interpMode=RCIM_Cubic, time=0, value=1}, <at 1 0>]}}}
       Scale Curve=10
      ))
    )) {isSetParameters=true}
   &ScaleSpriteSize(Scale Factor=&Vector2DFromCurve(
     Vector2Curve=DataInterface() {propertyValues {
       XCurve {keys [
         {arriveTangent=36.8345, interpMode=RCIM_Cubic, leaveTangent=36.8345, tangentMode=RCTM_User, time=0, value=0.2}
         {arriveTangent=-1.2585, interpMode=RCIM_Cubic, leaveTangent=-1.2585, tangentMode=RCTM_User, time=0.1, value=1.5}
         {arriveTangent=-1.6185, interpMode=RCIM_Cubic, leaveTangent=-1.6185, tangentMode=RCTM_User, time=0.5, value=0.75}
         {arriveTangent=-1.1243, interpMode=RCIM_Cubic, leaveTangent=-1.1243, tangentMode=RCTM_User, time=1, value=0.1}
        ]}
       YCurve {keys [
         {arriveTangent=0.5553, interpMode=RCIM_Cubic, leaveTangent=0.5553, tangentMode=RCTM_User, time=0, value=0.2}
         {arriveTangent=-1.7249, interpMode=RCIM_Cubic, leaveTangent=-1.7249, tangentMode=RCTM_User, time=0.5, value=0.75}
         {arriveTangent=-1.1243, interpMode=RCIM_Cubic, leaveTangent=-1.1243, tangentMode=RCTM_User, time=1, value=0.1}
        ]}
      }}
     ScaleCurve=<1 1>
    ))
   &self:Color_by_CameraAngle(
    Vector A:Vector3f=&ReturnCameraVector(Camera Property=::NewEnumerator1, Coordinate Space=::World)
    Vector B:Vector3f=&ReturnMeshOrientationAxis(Mesh Orientation Axis=::XAxis, Mesh Orientation Quat=$Particles.MeshOrientation)
   ) {enabled=false, module=Color_by_CameraAngle}
  ]
  rendererClasses [
   &NiagaraLightRendererProperties
   &NiagaraSpriteRendererProperties
   &NiagaraMeshRendererProperties
  ]
  renderers [
   {rendererClass=&NiagaraLightRendererProperties, rendererIndex=0}
   {rendererClass=&NiagaraSpriteRendererProperties, rendererIndex=1}
   {rendererClass=&NiagaraMeshRendererProperties, rendererIndex=2}
  ]
 }
 MuzzleFlash {
  simTarget=GPUComputeSim
  emitterUpdateScript [
   &EmitterState()
   &SpawnBurst_Instantaneous(Spawn Count=1, Use Spawn Probability=false)
  ]
  particleSpawnScript [
   /Niagara/Modules/Spawn/Initialization/V2/InitializeParticle(
    Color=<rgba 2 0.7992 0.0849 0.3>
    Lifetime=0.11
    Mesh Scale=<1 0.3 0.3>
    Mesh Scale Mode=::NewEnumerator4
    Sprite Size Mode=::NewEnumerator3
    Uniform Sprite Size=200
    UsePositionOffset=false
    Write Material Random=true
   ) {module=InitializeParticle001}
   &DynamicMaterialParameters(
    Index 0 Param 1=1
    Index 0 Param 2=5
    Module.Param0WriteEnabled=true
    Module.Param1WriteEnabled=true
    Module.Param2WriteEnabled=false
    Module.Param3WriteEnabled=false
   )
   /SetVariables(Particles.MeshOrientation=&FindQuatBetweenTwoVectors(From Vector=<0 0 0>, To Vector=<-1 0 0>)) {isSetParameters=true}
  ]
  particleUpdateScript [
   &ParticleState()
   &ScaleMeshSize(Scale Factor=&VectorFromCurve(VectorCurve=DataInterface() {propertyValues {
       XCurve {keys [
         {arriveTangent=2.0333, interpMode=RCIM_Cubic, leaveTangent=2.0333, tangentMode=RCTM_User, time=0, value=0.25}
         {arriveTangent=0.4705, interpMode=RCIM_Cubic, leaveTangent=0.4705, tangentMode=RCTM_User, time=1, value=1}
        ]}
       YCurve {keys [
         {arriveTangent=12.2668, interpMode=RCIM_Cubic, leaveTangent=12.2668, tangentMode=RCTM_User, time=0, value=0.25}
         {arriveTangent=-0.5545, interpMode=RCIM_Cubic, leaveTangent=-0.5545, tangentMode=RCTM_User, time=0.2, value=1}
         {arriveTangent=-1.8419, interpMode=RCIM_Cubic, leaveTangent=-1.8419, tangentMode=RCTM_User, time=1, value=0}
        ]}
       ZCurve {keys [
         {arriveTangent=12.2668, interpMode=RCIM_Cubic, leaveTangent=12.2668, tangentMode=RCTM_User, time=0, value=0.25}
         {arriveTangent=-0.5545, interpMode=RCIM_Cubic, leaveTangent=-0.5545, tangentMode=RCTM_User, time=0.2, value=1}
         {arriveTangent=-1.8419, interpMode=RCIM_Cubic, leaveTangent=-1.8419, tangentMode=RCTM_User, time=1, value=0}
        ]}
      }}))
   &ScaleColor(
    Scale Alpha=&FloatFromCurve(FloatCurve=DataInterface() {propertyValues {Curve {keys [<at 0.15 1 RCIM_Cubic>,<at 1 0>]}}})
    ScaleA=true
    ScaleRGB=true
   )
   &self:Color_by_CameraAngle(
    Vector A:Vector3f=&ReturnCameraVector(Camera Property=::NewEnumerator1, Coordinate Space=::World)
    Vector B:Vector3f=&ReturnMeshOrientationAxis(Mesh Orientation Axis=::XAxis, Mesh Orientation Quat=$Particles.MeshOrientation)
   ) {enabled=false, module=Color_by_CameraAngle}
  ]
  rendererClasses [&NiagaraMeshRendererProperties]
  renderers [{rendererClass=&NiagaraMeshRendererProperties, rendererIndex=0}]
 }
 sparkes {
  simTarget=GPUComputeSim
  emitterUpdateScript [
   &EmitterState()
   &SpawnBurst_Instantaneous(Spawn Count=90, Use Spawn Probability=false, UseLoopCountLimit=false)
  ]
  particleSpawnScript [
   /SetVariables(Particles.VelRand:NiagaraFloat=&FloatFromCurve(
     FloatCurve=DataInterface() {propertyValues {Curve {keys [
         {interpMode=RCIM_Cubic, tangentMode=RCTM_User, time=0, value=0}
         {arriveTangent=0.8678, interpMode=RCIM_Cubic, leaveTangent=0.8678, tangentMode=RCTM_User, time=0.9, value=0.25}
         {arriveTangent=1.9152, interpMode=RCIM_Cubic, leaveTangent=1.9152, tangentMode=RCTM_User, time=1, value=1}
        ]}}}
     CurveIndex=&RandomRangeFloat(Minimum=0, Maximum=1)
    )) {isSetParameters=true}
   /Niagara/Modules/Spawn/Initialization/InitializeParticle(
    Color=<rgba 25 3.8042 0 1>
    Lifetime=&Lerp_Float(
     A=&UniformRangedFloat(Minimum=0.07, Maximum=0.4)
     B=&UniformRangedFloat(Minimum=0.75, Maximum=1.1)
     Alpha=$Particles.VelRand
    )
    Module.Write Color=true
    Module.Write Lifetime=true
    Module.Write Mass=false
    Module.Write Material Random=true
    Module.Write Position=false
    Module.Write Scale=false
    Module.Write SpriteRotation=true
    Module.Write SpriteSize=true
    Position=&SimulationPosition()
    Sprite Rotation=&UniformRangedFloat(Minimum=-180, Maximum=180)
    Sprite Size=&Vector2DFromFloat(Value=&UniformRangedFloat(Minimum=1.5, Maximum=2.5))
   )
   &ShapeLocation(
    Handle Radius=10
    Large Radius=20
    MaskSpawnGroup=false
    Override Seed=false
    Rotation Angle Type:ENiagara_AngleInput=::NewEnumerator0
    Rotation Axis=<0 1 0>
    Rotation Mode=::NewEnumerator1
    Shape Primitive=::NewEnumerator3
    UseOwnerScale=true
   )
   &AddVelocity(
    Cone Angle=10
    Module.UseConeFalloff=false
    Module.UseVelDistribution=false
    Override Seed=false
    Velocity Mode=::NewEnumerator2
    Velocity Speed=&RandomRangeFloat(Minimum=200, Maximum=1000)
   )
  ]
  particleUpdateScript [
   &ParticleState()
   /Niagara/Modules/Update/Forces/CurlNoiseForce(
    Noise Frequency=0.2
    Noise Quality / Cost=::NewEnumerator0
    Noise Strength=&FloatFromCurve(
     FloatCurve=DataInterface() {propertyValues {Curve {keys [<at 0 0>, {arriveTangent=1, interpMode=RCIM_Cubic, leaveTangent=1, time=0.4106, value=0.6448}, <at 1 1>]}}}
     Scale Curve=&RandomRangeFloat(Minimum=400, Maximum=3000)
    )
    Pan Noise=true
    Pan Noise Field=<0.5 -1 1>
   )
   &Drag(Drag=&RandomRangeFloat(Minimum=4, Maximum=6), Use Linear Drag=true, Use Rotational Drag=false)
   &ScaleColor(
    Scale Alpha=&Multiply_Float(
     A=&FloatFromCurve(FloatCurve=DataInterface() {propertyValues {Curve {keys [
          {arriveTangent=11.9433, interpMode=RCIM_Cubic, leaveTangent=11.9433, tangentMode=RCTM_User, time=0, value=0}
          <at 0.2 1 RCIM_Cubic>
          <at 1 1 RCIM_Cubic>
         ]}}})
     B=&Multiply_Float(A=1, B=&FNTimeOfDayScalar(Night Scale=0.5, Morning Scale=0.25, Evening Scale=0.25))
    )
    ScaleA=true
    ScaleRGB=false
   ) {module=ScaleColor001}
   &DynamicMaterialParameters(
    Index 0 Param 1=&FloatFromCurve(FloatCurve=DataInterface() {propertyValues {Curve {keys [<at 0 0>,<at 1 1>]}}})
    Index 0 Param 2=1
    Index 0 Param 3=1
    Index 0 Param 4=1
    Module.Param0WriteEnabled=true
    Module.Param1WriteEnabled=true
    Module.Param2WriteEnabled=true
    Module.Param3WriteEnabled=true
   )
   &GravityForce(Gravity=&Lerp_Vector(A=&RandomRangeVector(Minimum=<0 0 -600>, Maximum=<0 0 -100>), B=<0 0 -1500>, Alpha=$Particles.VelRand))
   &Collision(
    Age Colliding Particles=false
    Friction=5
    GPU Collision Type=::NewEnumerator2
    Module.Control Rest State Via Surface Normal=false
    Module.EnableMaxCollisionCount=false
    Module.Randomize Collision Normal Vector=false
    Restitution=0.5
   )
   &SolveForcesAndVelocity(Clamp Velocity=false, Limit Acceleration=false)
   &ScaleSpriteSizeBySpeed(
    Max Scale Factor=&RandomRangeVector2D(Minimum=<0.4 0.5>, Maximum=<0.6 30>)
    Min Scale Factor=<0.3 0.5>
    Velocity Threshold:NiagaraFloat=&Lerp_Float(A=2500, B=600, Alpha=$Particles.VelRand)
   )
  ]
  rendererClasses [&NiagaraSpriteRendererProperties]
  renderers [{rendererClass=&NiagaraSpriteRendererProperties}]
 }
}
summary {
 systemName=NS_SawBlade_Muzzle
 emitters [
  {emitterName=MuzzleFlash, simTarget=GPUComputeSim, rendererClasses [&NiagaraMeshRendererProperties]}
  {
   emitterName=LightFlash
   rendererClasses [
    &NiagaraLightRendererProperties
    &NiagaraSpriteRendererProperties
    &NiagaraMeshRendererProperties
   ]
  }
  {emitterName=FN_Weapons_Smoke_BETA001, rendererClasses [&NiagaraSpriteRendererProperties]}
  {emitterName=sparkes, simTarget=GPUComputeSim, rendererClasses [&NiagaraSpriteRendererProperties]}
 ]
 userVariables [
  User.EndPoint {type=Vector3f, description="", defaultValue {struct=Vector3f, value=<0 0 0>}}
  User.WarmupTime {type=NiagaraFloat, defaultValue=0, description=""}
 ]
}
