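# Looping Niagara effect for the Sawblade Gun's blade, layering dust, debris, dirt, heat core, ribbon trail, sand, smoke and sparks.
# User.bIsAirborne gates spawn probability, limiting ground effects to contact, enabling Sparks in the air, and changing the ribbon's lifetime and velocity.
# User.PhysicalSurface feeds physical-material modules that set per-surface response for debris, dirt, heat core and smoke, with inline HLSL adjusting some offsets and lifetimes.
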NiagaraSystem /SawbladeGun/FX/NS_SawbladeGun_Blade_Sparks_loop
dependencies {
 usedDataInterfaces [&NiagaraDataInterfaceCurve, &NiagaraDataInterfaceVector2DCurve, &NiagaraDataInterfaceColorCurve]
 usedDynamicInputs []
 usedModules [&SystemState]
}
emitters {
 DustTrail {
  emitterUpdateScript [
   &EmitterState(Delay First Loop Only=true, Life Cycle Mode=::NewEnumerator1, Loop Delay=0.1, Loop Duration=1, UseLoopDelay=true)
   &SpawnRate(
    Spawn Probability=&MakeCustomFloatFromBool(Bool=$User.bIsAirborne, True Float=0, False Float=1)
    SpawnRate=20
    Use Spawn Probability=true
   )
   &SpawnPerUnit(
    Max Movement Threshold=2500
    Module.Use Max Movement Threshold=true
    Module.Use Movement Tolerance=true
    Movement Tolerance=10
    Spawn Probability:NiagaraFloat=&MakeCustomFloatFromBool(Bool=$User.bIsAirborne, True Float=0, False Float=1)
    Spawn Spacing=100
    Use Spawn Probability=true
   ) {enabled=false}
  ]
  particleSpawnScript [
   /SetVariables(Particles.VelCurveRand:NiagaraFloat=&FloatFromCurve(
     FloatCurve=DataInterface() {propertyValues {Curve {keys [
         {interpMode=RCIM_Cubic, leaveTangent=0.1132, tangentMode=RCTM_Break, time=0, value=0}
         {arriveTangent=1.9858, interpMode=RCIM_Cubic, leaveTangent=1.7203, tangentMode=RCTM_Break, time=0.3, value=0.4}
         {interpMode=RCIM_Cubic, tangentMode=RCTM_Break, time=1, value=1}
        ]}}}
     CurveIndex=&RandomRangeFloat(Minimum=0, Maximum=1)
    )) {isSetParameters=true}
   /SetVariables(Particles.InitSpriteRotRate=&RandomRangeFloat(Minimum=-1, Maximum=1)) {isSetParameters=true}
   &InitializeParticle(
    Color Maximum=<rgba 0.8073 0.7575 0.7245 0.5>
    Color Minimum=<rgba 0.3229 0.303 0.2898 0.5>
    Color Mode=::NewEnumerator2
    Lifetime Max=1.8
    Lifetime Min=1.2
    Lifetime Mode=::NewEnumerator1
    Sprite Rotation Angle=&RandomRangeFloat(Minimum=0, Maximum=360)
    Sprite Rotation Mode=::NewEnumerator0
    Sprite Size Mode=::NewEnumerator1
    Uniform Sprite Size Max=175
    Uniform Sprite Size Min=125
    UsePositionOffset=false
    Write Material Random=true
   )
   &ShapeLocation(
    Disc Coverage=1
    MaskSpawnGroup=false
    Override Seed=false
    Ring Radius=35
    Rotation Angle Type:ENiagara_AngleInput=::NewEnumerator0
    Rotation Axis=<0 1 0>
    Rotation Mode=::NewEnumerator1
    Shape Primitive=::NewEnumerator4
    UseOwnerScale=true
   )
   &InheritVelocity(Inherited Velocity Speed Limit=2000, Speed Limit=true, UseSpeedThreshold=false)
  ]
  particleUpdateScript [
   &ParticleState()
   &Drag(Drag=2, Use Linear Drag=true, Use Rotational Drag=false)
   &SpriteRotationRate(Rotation Rate=&Multiply_Float(
     A=$Particles.InitSpriteRotRate
     B=&FloatFromCurve(
      FloatCurve=DataInterface() {propertyValues {Curve {keys [
          {arriveTangent=-4.4624, interpMode=RCIM_Cubic, leaveTangent=-4.4624, tangentMode=RCTM_User, time=0, value=1}
          {arriveTangentWeight=0.1718, interpMode=RCIM_Cubic, tangentMode=RCTM_User, time=0.5, value=0.15}
         ]}}}
      Scale Curve=1500
     )
    ))
   &ScaleColor(
    Scale Alpha=&FloatFromCurve(
     FloatCurve=DataInterface() {propertyValues {Curve {keys [
         {interpMode=RCIM_Cubic, tangentMode=RCTM_User, time=0, value=0.1}
         {interpMode=RCIM_Cubic, tangentMode=RCTM_User, time=0.2, value=1}
         <at 1 1>
        ]}}}
     Scale Curve=&RandomRangeFloat(Minimum=0.1, Maximum=0.3)
    )
    Scale RGB=&VectorFromFloat(Value=1)
    ScaleA=true
    ScaleRGB=false
   )
   &ScaleSpriteSize(Scale Factor=&Vector2DFromFloat(Value=&FloatFromCurve(
      FloatCurve=DataInterface() {propertyValues {Curve {keys [
          {arriveTangent=0.0418, interpMode=RCIM_Cubic, leaveTangent=0.0418, tangentMode=RCTM_User, time=0, value=0.5}
          {arriveTangent=0.8771, interpMode=RCIM_Cubic, leaveTangent=0.8771, tangentMode=RCTM_User, time=0.7, value=0.85}
          {arriveTangentWeight=0.3375, interpMode=RCIM_Cubic, tangentMode=RCTM_User, time=1, value=1}
         ]}}}
      Scale Curve=0.9
     )))
   &DynamicMaterialParameters(
    Index 0 Param 1=&Add_Float(
     A=&RandomRangeFloat(Minimum=0, Maximum=1)
     B=&FloatFromCurve(
      FloatCurve=DataInterface() {propertyValues {Curve {keys [
          {arriveTangent=2.7533, interpMode=RCIM_Cubic, leaveTangent=2.5749, tangentMode=RCTM_Break, time=0, value=0}
          {arriveTangent=0.0529, interpMode=RCIM_Cubic, leaveTangent=0.2021, tangentMode=RCTM_User, time=1, value=1}
         ]}}}
      Scale Curve=0.2
     )
    )
    Index 0 Param 3=&FloatFromCurve(FloatCurve=DataInterface() {propertyValues {Curve {keys [{arriveTangent=-53.9199, time=0, value=0}, {leaveTangent=0.2723, time=0.25, value=0}]}}})
    Index 0 Param 4=&FloatFromCurve(FloatCurve=DataInterface() {propertyValues {Curve {keys [<at 0 1>,<at 1 0>]}}})
    Index 1 Param 1=&FloatFromCurve(FloatCurve=DataInterface() {propertyValues {Curve {keys [<at 0 0>, {arriveTangent=4.2257, interpMode=RCIM_Cubic, leaveTangent=4.2257, tangentMode=RCTM_User, time=0.25, value=0}]}}})
    Index 1 Param 2=&FloatFromCurve(FloatCurve=DataInterface() {propertyValues {Curve {keys [<at 0 0>]}}})
    Index 1 Param 3=&FloatFromCurve(FloatCurve=DataInterface() {propertyValues {Curve {keys [{interpMode=RCIM_Cubic, leaveTangent=6.3917, tangentMode=RCTM_Break, time=0, value=0.1}]}}})
    Index 1 Param 4=&FloatFromCurve(FloatCurve=DataInterface() {propertyValues {Curve {keys [{arriveTangent=-0.136, leaveTangent=-0.136, time=0, value=1}]}}})
    Index 2 Param 1=&FloatFromCurve(
     FloatCurve=DataInterface() {propertyValues {Curve {keys [
         {arriveTangent=4.9626, interpMode=RCIM_Cubic, leaveTangent=2.5739, tangentMode=RCTM_User, time=0, value=0}
         {leaveTangent=-0.0719, time=1, value=1}
        ]}}}
     Scale Curve=3
    )
    Index 2 Param 2=&FloatFromCurve(FloatCurve=DataInterface() {propertyValues {Curve {keys [<at 0 0.5>,<at 1 1>]}}})
    Index 2 Param 3=&FloatFromCurve(
     FloatCurve=DataInterface() {propertyValues {Curve {keys [
         {interpMode=RCIM_Cubic, leaveTangent=-6.6324, leaveTangentWeight=0.1676, tangentMode=RCTM_Break, time=0, value=1}
         {interpMode=RCIM_Cubic, tangentMode=RCTM_User, time=0.2, value=0}
         {interpMode=RCIM_Cubic, leaveTangent=11.1233, tangentMode=RCTM_Break, time=1, value=1}
        ]}}}
     Scale Curve=5
    )
    Index 2 Param 4=&RandomRangeFloat(Minimum=0.9, Maximum=1.4)
    Module.Param0WriteEnabled=true
    Module.Param0WriteEnabled1=true
    Module.Param0WriteEnabled2=true
    Module.Param1WriteEnabled=false
    Module.Param1WriteEnabled1=true
    Module.Param1WriteEnabled2=true
    Module.Param2WriteEnabled=true
    Module.Param2WriteEnabled1=true
    Module.Param2WriteEnabled2=true
    Module.Param3WriteEnabled=true
    Module.Param3WriteEnabled1=true
    Module.Param3WriteEnabled2=true
    Write Parameter Index 1=true
    Write Parameter Index 2=true
   )
   &SolveForcesAndVelocity(Clamp Velocity=false, Limit Acceleration=false)
  ]
  rendererClasses [&NiagaraSpriteRendererProperties]
  renderers [
   {rendererClass=&NiagaraSpriteRendererProperties}
   {rendererClass=&NiagaraSpriteRendererProperties, rendererIndex=1}
  ]
 }
 NE_FN_PhysMatDebris {
  emitterUpdateScript [
   &EmitterState()
   &SpawnBurst_Instantaneous(
    Spawn Count=&RandomRangeInt(Minimum=5, Maximum=10)
    Spawn Time=0.05
    Use Spawn Probability=false
    UseLoopCountLimit=false
   ) {enabled=false}
   &SpawnRate(
    Spawn Probability=&MakeCustomFloatFromBool(Bool=$User.bIsAirborne, True Float=0, False Float=1)
    SpawnRate=12
    Use Spawn Probability=true
   )
  ]
  particleSpawnScript [
   /SetVariables(Particles.BiasedRandom=&FloatFromCurve(
     FloatCurve=DataInterface() {propertyValues {Curve {keys [
         <at 0.2 1 RCIM_Cubic>
         {arriveTangent=-3.3394, interpMode=RCIM_Cubic, leaveTangent=-3.3394, tangentMode=RCTM_User, time=0.5, value=0.5}
         <at 0.8 0 RCIM_Cubic>
        ]}}}
     CurveIndex=&RandomRangeFloat(Minimum=0, Maximum=1)
    )) {isSetParameters=true, module=SetVariables_#1}
   &InitializeParticle(
    Color Mode=::NewEnumerator0
    Lifetime=&RandomRangeFloat(Minimum=0.5, Maximum=1.25)
    Mass=&Lerp_Float(A=0.6, B=1.2, Alpha=$Particles.BiasedRandom)
    Mass Mode=::NewEnumerator1
    Mesh Scale=&Multiply_VectorByFloat(
     Vector=&RandomRangeVector(Minimum=<0.9 0.4 0.5>, Maximum=<1 1 1>)
     Float=&Lerp_Float(A=0.7, B=1, Alpha=$Particles.BiasedRandom)
    )
    Mesh Scale Mode=::NewEnumerator4
    Position Offset=<-10 0 -30>
    UsePositionOffset=true
    Write Material Random=true
   )
   &FN_PhysicalSurface(
    Dirt Gravity=-900
    Dirt Mesh Count=4
    Dirt Velocity=200
    Dirt Velocity Cone Angle=1
    Glass Mesh Count=3
    Grass Gravity=-200
    Grass Scale=0.8
    Grass Velocity Cone Angle=30
    Ice Mesh Count=2
    Metal Gravity=&RandomRangeFloat(Minimum=-350, Maximum=-1800)
    Metal Mesh Count=2
    Metal Velocity=&RandomRangeFloat(Minimum=80, Maximum=375)
    Metal Velocity Cone Angle=130
    Physical Surface=$User.PhysicalSurface
    Stone Gravity=&RandomRangeFloat(Minimum=-700, Maximum=-1800)
    Stone Mesh Count=2
    Stone Velocity=&RandomRangeFloat(Minimum=80, Maximum=550)
    Stone Velocity Cone Angle=120
    StoneScale=1.5
    Wood Gravity=&RandomRangeFloat(Minimum=-450, Maximum=-900)
    Wood Mesh Count=2
    Wood Scale=2
    Wood Velocity=&RandomRangeFloat(Minimum=80, Maximum=600)
    Wood Velocity Cone Angle=130
   )
   &AddVelocityInCone(
    Cone Angle=$Output.FN_PhysicalSurface.SurfaceResponseConeAngle
    Module.Use Velocity Falloff On Cone Axis=false
    Velocity Strength=$Output.FN_PhysicalSurface.SurfaceResponseVelocity
   )
   &AddVelocityFromPoint(
    Module.Use Falloff Distance=false
    Origin Offset=<0 0 -50>
    Velocity Strength=&RandomRangeFloat(Minimum=80, Maximum=600)
   )
   &InitialMeshOrientation(Use Rotation Vector=false)
   &AddRotationalVelocity(Rotation Rate=&RandomVector(Vector Scale=&RandomRangeFloat(Minimum=1, Maximum=2)))
   &ApplyInitialForces()
   /SetVariables(Particles.GravityOverrideValue=$Output.FN_PhysicalSurface.GravityOverride) {isSetParameters=true, module=SetVariables_#2}
  ]
  particleUpdateScript [
   &ParticleState()
   &GravityForce(Gravity=$Particles.GravityOverrideValue)
   &Collision(
    Advanced Aging Rate=1
    Age Colliding Particles=false
    CPU Collision Trace Channel=ECC_WorldStatic
    Collision Enabled=&Boolean_AndOperation(
     Bool1=&SetBoolByFloatComparison(A=$Particles.Initial.Lifetime, B=1)
     Bool2=&SetBoolByFloatComparison(A=$Particles.Age, B=0.12)
    )
    Control Roll On Collision=true
    Friction=1
    Friction During a Bounce=0.25
    Max CPU Trace Length=100
    Module.Control Rest State Via Surface Normal=false
    Module.EnableMaxCollisionCount=false
    Module.Randomize Collision Normal Vector=true
    Randomize Collision Normal=0.25
    ScaleRateOfRotation=false
   ) {enabled=false}
   &SolveForcesAndVelocity(Clamp Velocity=false, Limit Acceleration=false)
   &ScaleMeshSize(Scale Factor=&VectorFromFloat(Value=&FloatFromCurve(FloatCurve=DataInterface() {propertyValues {Curve {keys [<at 0.6 1 RCIM_Cubic>, {arriveTangent=-6.308, interpMode=RCIM_Cubic, leaveTangent=-6.308, tangentMode=RCTM_User, time=1, value=0}]}}})))
   &DynamicMaterialParameters(Module.Param0WriteEnabled=true, Module.Param1WriteEnabled=true, Module.Param2WriteEnabled=false, Module.Param3WriteEnabled=false)
  ]
  rendererClasses [&NiagaraSpriteRendererProperties, &NiagaraMeshRendererProperties]
  renderers [
   {rendererClass=&NiagaraSpriteRendererProperties}
   {rendererClass=&NiagaraMeshRendererProperties, rendererIndex=1}
   {rendererClass=&NiagaraMeshRendererProperties, rendererIndex=2}
   {rendererClass=&NiagaraMeshRendererProperties, rendererIndex=3}
   {rendererClass=&NiagaraMeshRendererProperties, rendererIndex=4}
   {rendererClass=&NiagaraMeshRendererProperties, rendererIndex=5}
   {rendererClass=&NiagaraMeshRendererProperties, rendererIndex=6}
   {rendererClass=&NiagaraMeshRendererProperties, rendererIndex=7}
   {rendererClass=&NiagaraMeshRendererProperties, rendererIndex=8}
  ]
 }
 NE_FN_PhysMatDirt001 {
  emitterUpdateScript [
   &EmitterState(Life Cycle Mode=::NewEnumerator1, Loop Delay=0.04, Loop Duration=1, UseLoopDelay=false)
   &SpawnBurst_Instantaneous(Spawn Count=&RandomRangeInt(Minimum=5, Maximum=10), Spawn Time=0.05, Use Spawn Probability=false) {enabled=false}
   &SpawnRate(
    Spawn Probability=&MakeCustomFloatFromBool(Bool=$User.bIsAirborne, True Float=0, False Float=1)
    SpawnRate=&FloatFromCurve(
     FloatCurve=DataInterface() {propertyValues {Curve {keys [
         {arriveTangent=4.1785, interpMode=RCIM_Cubic, leaveTangent=4.1785, tangentMode=RCTM_User, time=0.06, value=0}
         {arriveTangent=5.4044, interpMode=RCIM_Cubic, leaveTangent=5.4044, tangentMode=RCTM_User, time=0.2, value=1}
        ]}}}
     CurveIndex=$Emitter.Age
     Scale Curve=40
    )
    Use Spawn Probability=true
   )
  ]
  particleSpawnScript [
   /SetVariables(Particles.BiasedRandom=&FloatFromCurve(
     FloatCurve=DataInterface() {propertyValues {Curve {keys [
         <at 0 1 RCIM_Cubic>
         {arriveTangent=-1.9207, interpMode=RCIM_Cubic, leaveTangent=-1.9207, tangentMode=RCTM_User, time=0.2, value=0.6}
         <at 0.8 0 RCIM_Cubic>
        ]}}}
     CurveIndex=&RandomRangeFloat(Minimum=0, Maximum=1)
    )) {isSetParameters=true, module=SetVariables_#1}
   &InitializeParticle(
    Color=<rgba 1 1 1 0>
    Lifetime=&RandomRangeFloat(Minimum=0.12, Maximum=0.12)
    Mass=&Lerp_Float(A=0.6, B=1.2, Alpha=$Particles.BiasedRandom)
    Mass Mode=::NewEnumerator1
    Mesh Scale=&Multiply_VectorByFloat(
     Vector=&RandomRangeVector(Minimum=<0.9 0.4 0.5>, Maximum=<1 1 1>)
     Float=&Lerp_Float(A=0.5, B=1.5, Alpha=$Particles.BiasedRandom)
    )
    Mesh Scale Mode=::NewEnumerator4
    UsePositionOffset=false
    Write Material Random=true
   )
   /Game/Effects/Niagara/Modules/PhysicalMaterial/FN_PhysicalMaterial(
    Default USE Particles Color=true
    Dirt Color:LinearColor=<rgba 0.125 0.0838 0.043 1>
    Dirt Visibility=2
    Grass Color:LinearColor=<rgba 0.125 0.0838 0.043 1>
    Grass Scale=0.5
    Grass Visibility=2
    Metal Gravity:NiagaraFloat=&RandomRangeFloat(Minimum=-350, Maximum=-1800)
    Metal Velocity:NiagaraFloat=&RandomRangeFloat(Minimum=80, Maximum=375)
    Metal Visibility=2
    MetalColor:LinearColor=<rgba 12 2.5609 0 0.5>
    MetalScale=0.4
    Physical Surface:NiagaraInt32=$User.PhysicalSurface
    Sand Color:LinearColor=<rgba 0.125 0.0838 0.043 1>
    Sand Mass=0.3
    Sand Scale=0.35
    Sand Visibility=2
    Stone Gravity:NiagaraFloat=&RandomRangeFloat(Minimum=-700, Maximum=-1250)
    Stone Velocity:NiagaraFloat=&RandomRangeFloat(Minimum=80, Maximum=550)
    StoneColor:LinearColor=<rgba 12 2.5609 0 0.5>
    StoneScale=0.4
    Watar Visibility=1
    Water Color:LinearColor=<rgba 0.6653 0.8741 1 1>
    Water Scale=0.9
    Wood Color:LinearColor=<rgba 0.125 0.0838 0.043 0.4>
    Wood Gravity:NiagaraFloat=&RandomRangeFloat(Minimum=-450, Maximum=-900)
    Wood Mesh Count=1
    Wood Scale=0.5
    Wood Velocity:NiagaraFloat=&RandomRangeFloat(Minimum=80, Maximum=600)
    Wood Visibility=2
   )
   &InitialMeshOrientation(
    Orientation Vector=&RandomRangeVector(Minimum=<-0.6 -0.5 0>, Maximum=<-1 0.5 1>)
    Rotation=&MakeVector(X=0, Y=0, Z=&Divide_Float(A=$User.InitRotZ, B=0))
    Rotation Coordinate Space=::NewEnumerator1
    Use Orientation Vector=true
    Use Rotation Vector=true
   )
   /SetVariables(Particles.GravityOverrideValue=$Output.FN_PhysicalMaterial.GravityOverride) {isSetParameters=true, module=SetVariables_#2}
   &BoxLocation(
    Box Offset=HlslExpression() {hlslExpression="User.PhysicalSurface ==8? float3(0.0, 0.0, 4.0) : float3(0.0, 0.0, -30.0)"}
    Box Size=<40 0 0>
    MaskSpawnGroup=false
    Module.Override Seed=false
    Rotate Box=false
    Surface Only=false
   )
   &CameraOffset(Camera Offset Amount=8)
   /SetVariables(Particles.Lifetime=&Multiply_Float(A=$Particles.Lifetime, B=HlslExpression() {hlslExpression="User.PhysicalSurface ==26? 0 : 1.0"})) {isSetParameters=true}
  ]
  particleUpdateScript [
   &ParticleState()
   &SolveForcesAndVelocity(Clamp Velocity=false, Limit Acceleration=false)
   &SolveRotationalForcesAndVelocity()
   &ScaleMeshSize(Scale Factor=&VectorFromFloat(Value=&FloatFromCurve(
      FloatCurve=DataInterface() {propertyValues {Curve {keys [
          {arriveTangent=4.1785, interpMode=RCIM_Cubic, leaveTangent=4.1785, tangentMode=RCTM_User, time=0, value=0}
          {arriveTangent=0.9995, interpMode=RCIM_Cubic, leaveTangent=0.9995, tangentMode=RCTM_User, time=0.3, value=0.5}
          {arriveTangent=-0.0405, interpMode=RCIM_Cubic, leaveTangent=-0.0405, tangentMode=RCTM_User, time=1, value=1}
         ]}}}
      Scale Curve=1.5
     )))
   &ScaleColor(
    Scale RGB=&VectorFromFloat(Value=&Lerp_Float(
      A=0.2
      B=5
      Alpha=&FloatFromCurve(
       FloatCurve=DataInterface() {propertyValues {Curve {keys [
           {interpMode=RCIM_Cubic, tangentMode=RCTM_User, time=0.4, value=0}
           {arriveTangent=1.7522, interpMode=RCIM_Cubic, leaveTangent=1.7522, tangentMode=RCTM_User, time=0.9, value=0.35}
           {arriveTangent=10.7124, interpMode=RCIM_Cubic, leaveTangent=10.7124, tangentMode=RCTM_User, time=1, value=1}
          ]}}}
       Scale Curve=&RandomRangeFloat(Minimum=0, Maximum=1)
      )
     ))
    ScaleA=false
    ScaleRGB=true
   )
  ]
  rendererClasses [&NiagaraMeshRendererProperties]
  renderers [
   {rendererClass=&NiagaraMeshRendererProperties, rendererIndex=0}
   {rendererClass=&NiagaraMeshRendererProperties, rendererIndex=1}
   {rendererClass=&NiagaraMeshRendererProperties, rendererIndex=2}
  ]
 }
 NE_FN_PhysMatHeatCore {
  emitterUpdateScript [
   &EmitterState(Life Cycle Mode=::NewEnumerator1, Loop Delay=0.04, Loop Duration=1, UseLoopDelay=false)
   &SpawnBurst_Instantaneous(Spawn Count=&RandomRangeInt(Minimum=5, Maximum=10), Spawn Time=0.05, Use Spawn Probability=false) {enabled=false}
   &SpawnRate(
    Spawn Probability=&MakeCustomFloatFromBool(Bool=$User.bIsAirborne, True Float=0, False Float=1)
    SpawnRate=&FloatFromCurve(
     FloatCurve=DataInterface() {propertyValues {Curve {keys [
         {arriveTangent=4.1785, interpMode=RCIM_Cubic, leaveTangent=4.1785, tangentMode=RCTM_User, time=0.06, value=0}
         {arriveTangent=5.4044, interpMode=RCIM_Cubic, leaveTangent=5.4044, tangentMode=RCTM_User, time=0.2, value=1}
        ]}}}
     CurveIndex=$Emitter.Age
     Scale Curve=115
    )
    Use Spawn Probability=true
   )
   &SpawnPerUnit(
    Max Movement Threshold=2500
    Module.Use Max Movement Threshold=true
    Module.Use Movement Tolerance=true
    Spawn Spacing=10
    Use Spawn Probability=false
   ) {enabled=false}
  ]
  particleSpawnScript [
   /SetVariables(Particles.BiasedRandom=&FloatFromCurve(
     FloatCurve=DataInterface() {propertyValues {Curve {keys [
         <at 0 1 RCIM_Cubic>
         {arriveTangent=-1.9207, interpMode=RCIM_Cubic, leaveTangent=-1.9207, tangentMode=RCTM_User, time=0.2, value=0.6}
         <at 0.8 0 RCIM_Cubic>
        ]}}}
     CurveIndex=&RandomRangeFloat(Minimum=0, Maximum=1)
    )) {isSetParameters=true, module=SetVariables_#1}
   &InitializeParticle(
    Color=<rgba 1 1 1 0>
    Lifetime=&RandomRangeFloat(Minimum=0.07, Maximum=0.07)
    Mass=&Lerp_Float(A=0.6, B=1.2, Alpha=$Particles.BiasedRandom)
    Mass Mode=::NewEnumerator1
    Mesh Scale=&Multiply_VectorByFloat(
     Vector=&RandomRangeVector(Minimum=<0.9 0.4 0.5>, Maximum=<1 1 1>)
     Float=&Lerp_Float(A=0.7, B=1, Alpha=$Particles.BiasedRandom)
    )
    Mesh Scale Mode=::NewEnumerator4
    UsePositionOffset=false
    Write Material Random=true
   )
   &FN_PhysicalMaterial(
    Default USE Particles Color=true
    Dirt Color:LinearColor=<rgba 12 2.5609 0 0.5>
    Dirt Scale=0.4
    Dirt Visibility=2
    Grass Color:LinearColor=<rgba 12 2.5609 0 0.5>
    Grass Scale=0.3
    Grass Visibility=2
    Metal Gravity:NiagaraFloat=&RandomRangeFloat(Minimum=-350, Maximum=-1800)
    Metal Velocity:NiagaraFloat=&RandomRangeFloat(Minimum=80, Maximum=375)
    Metal Visibility=2
    MetalColor:LinearColor=<rgba 12 2.5609 0 0.5>
    MetalScale=0.4
    Physical Surface:NiagaraInt32=$User.PhysicalSurface
    Sand Color:LinearColor=<rgba 12 2.5609 0 0.5>
    Sand Mass=0.3
    Sand Scale=0.35
    Sand Visibility=2
    Stone Gravity:NiagaraFloat=&RandomRangeFloat(Minimum=-700, Maximum=-1250)
    Stone Velocity:NiagaraFloat=&RandomRangeFloat(Minimum=80, Maximum=550)
    StoneColor:LinearColor=<rgba 12 2.5609 0 0.5>
    StoneScale=0.3
    Wood Color:LinearColor=<rgba 12 2.5609 0 0.5>
    Wood Gravity:NiagaraFloat=&RandomRangeFloat(Minimum=-450, Maximum=-900)
    Wood Mesh Count=1
    Wood Scale=0.25
    Wood Velocity:NiagaraFloat=&RandomRangeFloat(Minimum=80, Maximum=600)
    Wood Visibility=2
   )
   &InitialMeshOrientation(
    Orientation Vector=&RandomRangeVector(Minimum=<-0.6 -0.5 0>, Maximum=<-1 0.5 1>)
    Rotation=&MakeVector(X=0, Y=0, Z=&Divide_Float(A=$User.InitRotZ, B=0))
    Rotation Coordinate Space=::NewEnumerator1
    Use Orientation Vector=true
    Use Rotation Vector=true
   )
   /SetVariables(Particles.GravityOverrideValue=$Output.FN_PhysicalMaterial.GravityOverride) {isSetParameters=true, module=SetVariables_#2}
   &BoxLocation(Box Offset=<20 0 -30>, Box Size=<50 0 0>, MaskSpawnGroup=false, Module.Override Seed=false, Rotate Box=false, Surface Only=false)
   &CameraOffset(Camera Offset Amount=&RandomRangeFloat(Minimum=3, Maximum=5))
  ]
  particleUpdateScript [
   /SetVariables(Particles.VelLength:NiagaraFloat=&VectorLength(Vector=$Engine.Owner.Velocity)) {isSetParameters=true}
   &ParticleState()
   &SolveForcesAndVelocity(Clamp Velocity=false, Limit Acceleration=false)
   &SolveRotationalForcesAndVelocity()
   &ScaleMeshSize(Scale Factor=&VectorFromFloat(Value=&FloatFromCurve(
      FloatCurve=DataInterface() {propertyValues {Curve {keys [
          {arriveTangent=4.1785, interpMode=RCIM_Cubic, leaveTangent=4.1785, tangentMode=RCTM_User, time=0, value=0}
          {arriveTangent=0.9995, interpMode=RCIM_Cubic, leaveTangent=0.9995, tangentMode=RCTM_User, time=0.3, value=0.5}
          {arriveTangent=-0.0405, interpMode=RCIM_Cubic, leaveTangent=-0.0405, tangentMode=RCTM_User, time=1, value=1}
         ]}}}
      Scale Curve=5.5
     )))
   &ScaleColor(
    Scale RGB=&VectorFromFloat(Value=&Lerp_Float(
      A=0.2
      B=5
      Alpha=&FloatFromCurve(
       FloatCurve=DataInterface() {propertyValues {Curve {keys [
           {interpMode=RCIM_Cubic, tangentMode=RCTM_User, time=0.4, value=0}
           {arriveTangent=1.7522, interpMode=RCIM_Cubic, leaveTangent=1.7522, tangentMode=RCTM_User, time=0.9, value=0.35}
           {arriveTangent=10.7124, interpMode=RCIM_Cubic, leaveTangent=10.7124, tangentMode=RCTM_User, time=1, value=1}
          ]}}}
       Scale Curve=&RandomRangeFloat(Minimum=0, Maximum=1)
      )
     ))
    ScaleA=false
    ScaleRGB=true
   )
   &ScaleMeshSize(Scale Factor=&VectorFromFloat(Value=&Lerp_Float(
      A=1
      B=&RandomRangeFloat(Minimum=0.2, Maximum=1)
      Alpha=&MakeCustomFloatFromBool(
       Bool=&SetBoolByFloatComparison(A=$Particles.VelLength, B=300, Comparison Type=::NewEnumerator6)
       True Float=1
       False Float=0
      )
     ))) {module=ScaleMeshSize001}
   &ScaleColor(
    Scale RGB=&Lerp_Vector(
     A=<1 1 1>
     B=<1 1 1>
     Alpha=&MakeCustomFloatFromBool(
      Bool=&SetBoolByFloatComparison(A=$Particles.VelLength, B=300, Comparison Type=::NewEnumerator6)
      True Float=1
      False Float=0
     )
    )
    ScaleA=false
    ScaleRGB=true
   ) {module=ScaleColor001}
   &UpdateMeshOrientation(
    Facing Direction=&Lerp_Vector(
     A=<1 0 0>
     B=&RandomRangeVector(Minimum=<1 -0.5 0>, Maximum=<1 0.5 1>)
     Alpha=&MakeCustomFloatFromBool(
      Bool=&SetBoolByFloatComparison(A=$Particles.VelLength, B=300, Comparison Type=::NewEnumerator6)
      True Float=1
      False Float=0
     )
    )
    Facing Interpolation=false
    Orientation Method=::NewEnumerator1
    Override Quat=false
    TransformFacing=false
    TransformUp=false
   )
  ]
  rendererClasses [&NiagaraMeshRendererProperties]
  renderers [
   {rendererClass=&NiagaraMeshRendererProperties, rendererIndex=0}
   {rendererClass=&NiagaraMeshRendererProperties, rendererIndex=1}
   {rendererClass=&NiagaraMeshRendererProperties, rendererIndex=2}
  ]
 }
 RibbonTrail {
  emitterUpdateScript [
   &EmitterState(Delay First Loop Only=true, Life Cycle Mode=::NewEnumerator1, Loop Delay=0.25, Loop Duration=1, UseLoopDelay=true)
   &SpawnRate(SpawnRate=40, Use Spawn Probability=true)
  ]
  particleSpawnScript [
   &InitializeRibbon(
    Lifetime=&Multiply_Float(A=0.8, B=&MakeCustomFloatFromBool(Bool=$User.bIsAirborne, True Float=0.5, False Float=1))
    Ribbon Facing Mode:ENiagara_UnsetDirectSet=::NewEnumerator1
    Ribbon Width=150
    UsePositionOffset=false
    Write Material Random=true
   )
   &AddVelocity(Velocity=&Multiply_VectorByFloat(
     Vector=<-1500 0 0>
     Float=&MakeCustomFloatFromBool(Bool=$User.bIsAirborne, True Float=1, False Float=0)
    ))
  ]
  particleUpdateScript [
   &ParticleState()
   &SolveForcesAndVelocity(Clamp Velocity=false, Limit Acceleration=false)
   &DynamicMaterialParameters(
    Index 0 Param 1=&FloatFromCurve(FloatCurve=DataInterface() {propertyValues {Curve {keys [
         {interpMode=RCIM_Cubic, tangentMode=RCTM_User, time=0, value=1}
         {arriveTangent=-3.057, interpMode=RCIM_Cubic, leaveTangent=-3.057, tangentMode=RCTM_User, time=1, value=0}
        ]}}})
    Index 0 Param 2=&FloatFromCurve(FloatCurve=DataInterface() {propertyValues {Curve {keys [{interpMode=RCIM_Cubic, tangentMode=RCTM_User, time=0, value=1}, {interpMode=RCIM_Cubic, tangentMode=RCTM_User, time=0.5, value=0.1}]}}})
    Index 0 Param 3=1.5
    Index 0 Param 4=0.2
    Index 1 Param 1=400
    Module.Param0WriteEnabled=true
    Module.Param0WriteEnabled1=true
    Module.Param1WriteEnabled=true
    Module.Param1WriteEnabled1=true
    Module.Param2WriteEnabled=true
    Module.Param2WriteEnabled1=true
    Module.Param3WriteEnabled=true
    Module.Param3WriteEnabled1=true
    Write Parameter Index 1=true
   )
   &ScaleColor(
    Scale Alpha=&FloatFromCurve(FloatCurve=DataInterface() {propertyValues {Curve {keys [
         {interpMode=RCIM_Cubic, tangentMode=RCTM_User, time=0, value=0}
         {interpMode=RCIM_Cubic, tangentMode=RCTM_Break, time=0.05, value=1}
         {arriveTangent=-1.6091, interpMode=RCIM_Cubic, leaveTangent=-1.6091, tangentMode=RCTM_User, time=0.6, value=0.5}
         {arriveTangentWeight=0.2136, interpMode=RCIM_Cubic, tangentMode=RCTM_Break, time=1, value=0}
        ]}}})
    Scale RGB=&MakeVectorFromLinearColorRGB(Color=&ColorFromCurve(ColorCurve=DataInterface() {propertyValues {
        AlphaCurve {keys []}
        BlueCurve {keys [{arriveTangent=1.3532, interpMode=RCIM_Cubic, time=0, value=0}, {interpMode=RCIM_Cubic, leaveTangent=1.3532, time=0.2843, value=0.2969}]}
        GreenCurve {keys [{arriveTangent=0.3631, interpMode=RCIM_Cubic, time=0, value=0.3965}, {interpMode=RCIM_Cubic, leaveTangent=0.3631, time=0.2843, value=0.2969}]}
        RedCurve {keys [{arriveTangent=-3.205, interpMode=RCIM_Cubic, time=0, value=1}, {interpMode=RCIM_Cubic, leaveTangent=-3.205, time=0.2843, value=0.2969}]}
       }}))
    ScaleA=true
    ScaleRGB=true
   )
   &ScaleRibbonWidth(Ribbon Width Scale=&FloatFromCurve(FloatCurve=DataInterface() {propertyValues {Curve {keys [{interpMode=RCIM_Cubic, tangentMode=RCTM_User, time=0, value=1}, {interpMode=RCIM_Cubic, tangentMode=RCTM_User, time=1, value=0.5}]}}}))
  ]
  rendererClasses [&NiagaraRibbonRendererProperties]
  renderers [{rendererClass=&NiagaraRibbonRendererProperties}]
 }
 Sand {
  emitterUpdateScript [
   &EmitterState()
   &SpawnRate(SpawnRate=50, Use Spawn Probability=false) {enabled=false}
   &SpawnPerUnit(
    Max Movement Threshold=2500
    Module.Use Max Movement Threshold=true
    Module.Use Movement Tolerance=true
    Movement Tolerance=10
    Spawn Probability:NiagaraFloat=&MakeCustomFloatFromBool(Bool=$User.bIsAirborne, True Float=0, False Float=1)
    Spawn Spacing=50
    Use Spawn Probability=true
   )
  ]
  particleSpawnScript [
   &InitializeParticle(
    Color Maximum=<rgba 0.2708 0.2506 0.2327 1>
    Color Minimum=<rgba 0.0729 0.0681 0.063 1>
    Color Mode=::NewEnumerator2
    Lifetime Max=0.35
    Lifetime Min=0.2
    Lifetime Mode=::NewEnumerator1
    Sprite Size Max=<55 150>
    Sprite Size Min=<40 125>
    Sprite Size Mode=::NewEnumerator2
    UsePositionOffset=false
    Write Material Random=true
   )
   /Niagara/Modules/Spawn/Location/SphereLocation(
    MaskSpawnGroup=false
    Module.Override Randomness=false
    Non Uniform Scale=<2 1 1>
    Offset=HlslExpression() {hlslExpression="User.PhysicalSurface ==8? float3(0.0, 0.0, 0.0) : float3(0.0, 0.0, -15.0) /* Custom HLSL! */"}
    Sphere Radius=2
    Surface Only=false
   )
   &SubUVAnimation(End Frame=3, SubUV Animation Mode:ENiagara_SubUVLookupMode=::NewEnumerator2)
   &AddVelocity(
    Scale Added Velocity=HlslExpression() {hlslExpression="float3(1.0, sign(rand(1.0)-0.5), 1.0) /* Custom HLSL! */"}
    Velocity=&Multiply_VectorByFloat(Vector=&RandomRangeVector(Minimum=<-500 600 150>, Maximum=<-2200 1200 1250>), Float=0.1)
   ) {module=AddVelocity001}
  ]
  particleUpdateScript [
   &ParticleState()
   &Drag(Use Linear Drag=true, Use Rotational Drag=false)
   &SolveForcesAndVelocity(Clamp Velocity=false, Limit Acceleration=false)
   &ScaleSpriteSize(Scale Factor=&Vector2DFromCurve(
     Vector2Curve=DataInterface() {propertyValues {XCurve {keys [<at 0.0036 1.6897>,<at 0.9967 0.0582>]}, YCurve {keys [<at 0 1>,<at 1.002 3.7227>]}}}
     ScaleCurve=&Vector2DFromFloat(Value=0.85)
    ))
   &DynamicMaterialParameters(
    Index 0 Param 2=1
    Index 0 Param 3=1
    Index 0 Param 4=&FloatFromCurve(FloatCurve=DataInterface() {propertyValues {Curve {keys [
         {interpMode=RCIM_Cubic, leaveTangent=-2.8838, tangentMode=RCTM_Break, time=0.3, value=1}
         {arriveTangentWeight=0.5941, interpMode=RCIM_Cubic, tangentMode=RCTM_Break, time=1, value=0.25}
        ]}}})
    Module.Param0WriteEnabled=true
    Module.Param1WriteEnabled=true
    Module.Param2WriteEnabled=true
    Module.Param3WriteEnabled=true
   )
   &ScaleColor(
    Scale Alpha=&FloatFromCurve(
     FloatCurve=DataInterface() {propertyValues {Curve {keys [
         {interpMode=RCIM_Cubic, leaveTangentWeight=0.8777, tangentMode=RCTM_User, time=0, value=1}
         {interpMode=RCIM_Cubic, tangentMode=RCTM_User, time=0.2, value=1}
         {arriveTangentWeight=0.4812, interpMode=RCIM_Cubic, tangentMode=RCTM_User, time=1, value=0}
        ]}}}
     Scale Curve=&RandomRangeFloat(Minimum=0.35, Maximum=0.75)
    )
    ScaleA=true
    ScaleRGB=false
   )
  ]
  rendererClasses [&NiagaraSpriteRendererProperties]
  renderers [{rendererClass=&NiagaraSpriteRendererProperties}]
 }
 Smoke {
  emitterUpdateScript [
   &EmitterState()
   &SpawnRate(
    Spawn Probability=&MakeCustomFloatFromBool(Bool=$User.bIsAirborne, True Float=0, False Float=1)
    SpawnRate=40
    Use Spawn Probability=true
   )
   &SpawnPerUnit(
    Max Movement Threshold=2500
    Module.Use Max Movement Threshold=true
    Module.Use Movement Tolerance=true
    Movement Tolerance=10
    Spawn Probability:NiagaraFloat=&MakeCustomFloatFromBool(Bool=$User.bIsAirborne, True Float=0, False Float=1)
    Spawn Spacing=25
    Use Spawn Probability=true
   ) {enabled=false}
  ]
  particleSpawnScript [
   &InitializeParticle(
    Color=<rgba 0.5255 0.4708 0.4218 0.8>
    Lifetime=&UniformRangedFloat(Minimum=0.2, Maximum=0.3)
    Position Offset=<0 0 -20>
    Position Offset Coordinate Space=::World
    Sprite Size Mode=::NewEnumerator3
    Sprite UV Mode=::NewEnumerator3
    Uniform Sprite Size=&UniformRangedFloat(Minimum=20, Maximum=35)
    UsePositionOffset=true
    Write Material Random=true
   ) {module=InitializeParticle001}
   &FN_PhysicalMaterial(
    Default USE Particles Color=true
    Dirt Color:LinearColor=<rgba 0.6372 0.4353 0.2953 1>
    Dirt Gravity=1000
    Dirt Visibility=1
    Grass Color:LinearColor=<rgba 1 0.8863 0.7534 0.5>
    Grass Gravity=1000
    Grass Visibility=1
    Metal Gravity:NiagaraFloat=&RandomRangeFloat(Minimum=700, Maximum=1000)
    Metal Velocity:NiagaraFloat=&RandomRangeFloat(Minimum=80, Maximum=375)
    Metal Velocity Cone Angle=130
    Metal Visibility=1
    MetalColor:LinearColor=<rgba 1 1 1 0.5>
    Physical Surface:NiagaraInt32=$User.PhysicalSurface
    Sand Color:LinearColor=<rgba 0.6372 0.5317 0.427 1>
    Sand Gravity=1000
    Sand Scale=1.4
    Sand Visibility=1
    Stone Gravity:NiagaraFloat=&RandomRangeFloat(Minimum=1000, Maximum=1400)
    Stone Velocity:NiagaraFloat=&RandomRangeFloat(Minimum=80, Maximum=550)
    Stone Visibility=1
    StoneColor:LinearColor=<rgba 0.6823 0.5846 0.4523 0.6>
    Watar Visibility=1
    Water Color:LinearColor=<rgba 0.8761 1.0196 1.6 1>
    Water Gravity=1200
    Water Velocity Cone Angle=100
    Wood Color:LinearColor=<rgba 1 1 1 0.5>
    Wood Gravity:NiagaraFloat=&RandomRangeFloat(Minimum=550, Maximum=900)
    Wood Mesh Count=1
    Wood Velocity:NiagaraFloat=&RandomRangeFloat(Minimum=80, Maximum=600)
    Wood Velocity Cone Angle=130
   )
   /SetVariables(Particles.InitSpriteRotRate=&UniformRangedFloat(Minimum=-1, Maximum=1)) {isSetParameters=true}
   /SetVariables(
    Particles.GravityOverrideValue=$Output.FN_PhysicalMaterial.GravityOverride
    Particles.InitRand=HlslExpression() {hlslExpression="sign(rand(1.0)-0.5)"}
   ) {isSetParameters=true}
   &AddVelocity(
    Scale Added Velocity=HlslExpression() {hlslExpression="float3(1.0, Particles.InitRand, 1.0) /* Custom HLSL! */"}
    Velocity=&RandomRangeVector(Minimum=<-100 300 0>, Maximum=<-500 500 50>)
   ) {enabled=false, module=AddVelocity002}
   &AddVelocity(
    Scale Added Velocity=HlslExpression() {hlslExpression="float3(1.0, Particles.InitRand, 1.0) /* Custom HLSL! */"}
    Velocity=&RandomRangeVector(Minimum=<-100 100 0>, Maximum=<-300 300 50>)
   )
   &ShapeLocation(
    MaskSpawnGroup=false
    Offset=&Multiply_Vector(
     A=HlslExpression() {hlslExpression="User.PhysicalSurface ==8? float3(0.0, 15.0, 0.0) : float3(0.0, 15.0, -25.0) /* Custom HLSL! */"}
     B=HlslExpression() {hlslExpression="float3(1.0, Particles.InitRand, 1.0) "}
    )
    Offset Mode=::NewEnumerator0
    Override Seed=false
    Ring Radius=3
    Shape Primitive=::NewEnumerator4
    UseOwnerScale=true
   )
   &ApplyInitialForces()
   &CameraOffset(Camera Offset Amount=&RandomRangeFloat(Minimum=0, Maximum=2))
  ]
  particleUpdateScript [
   &ParticleState()
   &SpriteRotationRate(Rotation Rate=&Multiply_Float(
     A=$Particles.InitSpriteRotRate
     B=&FloatFromCurve(
      FloatCurve=DataInterface() {propertyValues {Curve {keys [
          {arriveTangent=-1.9214, interpMode=RCIM_Cubic, leaveTangent=-1.9214, tangentMode=RCTM_User, time=0.2, value=1.0113}
          {arriveTangent=-0.2496, interpMode=RCIM_Cubic, leaveTangent=-0.2496, tangentMode=RCTM_User, time=0.9, value=0.16}
         ]}}}
      CurveIndex=$Emitter.Age
      Scale Curve=10
     )
    ))
   &ScaleColor(
    Scale Alpha=&FloatFromCurve(FloatCurve=DataInterface() {propertyValues {Curve {keys [<at 0 0 RCIM_Cubic>,<at 0.15 0.9765>,<at 0.7 0.9107>,<at 1 0>]}}}, Scale Curve=1.2)
    ScaleA=true
    ScaleRGB=true
   ) {module=ScaleColor001}
   &Drag(Drag=5, Use Linear Drag=true, Use Rotational Drag=false)
   &GravityForce(Gravity=$Particles.GravityOverrideValue)
   &SolveForcesAndVelocity(Clamp Velocity=false, Limit Acceleration=false)
   &ScaleSpriteSize(Scale Factor=&Vector2DFromFloat(Value=&FloatFromCurve(
      FloatCurve=DataInterface() {propertyValues {Curve {keys [
          {arriveTangent=1.5594, interpMode=RCIM_Cubic, leaveTangent=1.5594, tangentMode=RCTM_User, time=-0.0046, value=0.5}
          <at 0.3 1.4923>
          <at 1.0011 3.4809 RCIM_Cubic>
         ]}}}
      Scale Curve=2
     )))
   &DynamicMaterialParameters(
    Index 0 Param 1=&FloatFromCurve(FloatCurve=DataInterface() {propertyValues {Curve {keys [<at 0 0 RCIM_Cubic>,<at 1.0011 0.2>]}}})
    Index 0 Param 2=&FloatFromCurve(FloatCurve=DataInterface() {propertyValues {Curve {keys [
         {arriveTangent=0.0453, interpMode=RCIM_Cubic, tangentMode=RCTM_Break, time=0, value=-1}
         {arriveTangent=0.0604, interpMode=RCIM_Cubic, leaveTangent=0.0604, tangentMode=RCTM_User, time=1, value=-1}
        ]}}})
    Index 0 Param 3=&FloatFromCurve(FloatCurve=DataInterface() {propertyValues {Curve {keys [{interpMode=RCIM_Cubic, leaveTangent=4.6779, time=0.2851, value=0}]}}})
    Index 1 Param 1=&FloatFromCurve(
     FloatCurve=DataInterface() {propertyValues {Curve {keys [<at 0 0>,<at 0.25 1>]}}}
     Scale Curve=&RandomRangeFloat(Minimum=2, Maximum=1)
    )
    Index 1 Param 2=&FloatFromCurve(
     FloatCurve=DataInterface() {propertyValues {Curve {keys [
         {interpMode=RCIM_Cubic, leaveTangent=3.7422, tangentMode=RCTM_Break, time=0, value=0.2}
         {interpMode=RCIM_Cubic, leaveTangent=0.9066, time=1, value=2}
        ]}}}
     Scale Curve=&RandomRangeFloat(Minimum=0.5, Maximum=1)
    )
    Index 1 Param 3=&FloatFromCurve(FloatCurve=DataInterface() {propertyValues {Curve {keys [
         {interpMode=RCIM_Cubic, leaveTangent=2.8445, tangentMode=RCTM_Break, time=0, value=0.2}
         {interpMode=RCIM_Cubic, leaveTangent=63.2978, time=0.25, value=1}
        ]}}})
    Index 1 Param 4=&FloatFromCurve(
     FloatCurve=DataInterface() {propertyValues {Curve {keys [<at 0 1>,<at 1 0.8>]}}}
     Scale Curve=&RandomRangeFloat(Minimum=0.5, Maximum=1)
    )
    Index 2 Param 1=&FloatFromCurve(FloatCurve=DataInterface() {propertyValues {Curve {keys [{interpMode=RCIM_Cubic, leaveTangent=3.2309, tangentMode=RCTM_Break, time=0.1, value=0}, <at 1 1>]}}})
    Index 2 Param 2=$Particles.MaterialRandom
    Index 2 Param 3=&FloatFromCurve(FloatCurve=DataInterface() {propertyValues {Curve {keys [
         {interpMode=RCIM_Cubic, tangentMode=RCTM_Break, time=0, value=1}
         {interpMode=RCIM_Cubic, leaveTangent=11.1233, tangentMode=RCTM_Break, time=1, value=2}
        ]}}})
    Index 2 Param 4=0.75
    Module.Param0WriteEnabled=true
    Module.Param0WriteEnabled1=true
    Module.Param0WriteEnabled2=true
    Module.Param1WriteEnabled=false
    Module.Param1WriteEnabled1=true
    Module.Param1WriteEnabled2=true
    Module.Param2WriteEnabled=true
    Module.Param2WriteEnabled1=true
    Module.Param2WriteEnabled2=true
    Module.Param3WriteEnabled=true
    Module.Param3WriteEnabled1=true
    Module.Param3WriteEnabled2=true
    Write Parameter Index 1=true
    Write Parameter Index 2=true
   )
   /SetVariables(Particles.Test:NiagaraFloat=$Emitter.SpawnPerUnit.SpawnRemainder) {enabled=false, isSetParameters=true}
  ]
  rendererClasses [&NiagaraSpriteRendererProperties]
  renderers [
   {rendererClass=&NiagaraSpriteRendererProperties}
   {rendererClass=&NiagaraSpriteRendererProperties, rendererIndex=1}
  ]
 }
 Sparks {
  emitterUpdateScript [
   &EmitterState()
   &SpawnRate(
    Spawn Probability=&MakeCustomFloatFromBool(Bool=$User.bIsAirborne, True Float=1, False Float=0)
    SpawnRate=50
    Use Spawn Probability=true
   )
  ]
  particleSpawnScript [
   &InitializeParticle(
    Color=&LerpLinearColors(
     StartColor=<rgba 10 3.8608 0.502 1>
     EndColor=<rgba 0.4427 0.029 0 1>
     LerpFactor=&FloatFromCurve(
      FloatCurve=DataInterface() {propertyValues {Curve {keys [
          {interpMode=RCIM_Cubic, leaveTangentWeight=1.1961, tangentMode=RCTM_User, time=0, value=0}
          {arriveTangentWeight=0.1113, interpMode=RCIM_Cubic, tangentMode=RCTM_User, time=1, value=1}
         ]}}}
      CurveIndex=&RandomRangeFloat(Minimum=0, Maximum=1)
     )
    )
    Lifetime Max=0.3
    Lifetime Min=0.1
    Lifetime Mode=::NewEnumerator1
    Sprite Size Mode=::NewEnumerator3
    Uniform Sprite Size=&Lerp_Float(
     A=1.5
     B=2.5
     Alpha=&FloatFromCurve(
      FloatCurve=DataInterface() {propertyValues {Curve {keys [
          {interpMode=RCIM_Cubic, tangentMode=RCTM_User, time=0, value=0}
          {arriveTangent=2.3538, interpMode=RCIM_Cubic, leaveTangent=2.3538, tangentMode=RCTM_User, time=1, value=1}
         ]}}}
      CurveIndex=&RandomRangeFloat(Minimum=0, Maximum=1)
     )
    )
    UsePositionOffset=false
    Write Material Random=true
   )
   &ShapeLocation(
    Handle Radius=10
    Large Radius=25
    MaskSpawnGroup=false
    Override Seed=false
    Rotation Axis=<0 1 0>
    Rotation Mode=::NewEnumerator1
    Shape Primitive=::NewEnumerator3
    UseOwnerScale=true
   )
  ]
  particleUpdateScript [
   &ParticleState()
   &VortexForce(
    UseOriginOffset=false
    Vortex Axis=<0 1 0>
    Vortex Axis Coordinate Space:ENiagaraCoordinateSpace=::Local
    Vortex Force Amount=&RandomRangeFloat(Minimum=2000, Maximum=4000)
   )
   &AccelerationForce(Acceleration=&RandomRangeVector(Minimum=<0 0 -1200>, Maximum=<0 0 -500>)) {enabled=false}
   &Drag(Drag=2, Use Linear Drag=true, Use Rotational Drag=false)
   &SolveForcesAndVelocity(Clamp Velocity=false, Limit Acceleration=false)
   &ScaleColor(
    Linear Color Curve=DataInterface() {propertyValues {
      AlphaCurve {keys [<at 0 1>,<at 1 1>]}
      BlueCurve {keys [
        <at 0 2 RCIM_Cubic>
        {arriveTangent=-1.452, interpMode=RCIM_Cubic, leaveTangent=-1.452, tangentMode=RCTM_User, time=0.1467, value=1}
        <at 1 0>
       ]}
      GreenCurve {keys [
        <at 0 2 RCIM_Cubic>
        {arriveTangent=-1.4408, interpMode=RCIM_Cubic, leaveTangent=-1.4408, tangentMode=RCTM_User, time=0.1467, value=1}
        <at 1 0.0349>
       ]}
      RedCurve {keys [
        <at 0 2 RCIM_Cubic>
        {arriveTangent=-0.8332, interpMode=RCIM_Cubic, leaveTangent=-0.8332, tangentMode=RCTM_User, time=0.1467, value=1}
        <at 1 0.2969>
       ]}
     }}
    Scale Mode=::NewEnumerator2
   ) {module=ScaleColor001}
   &ScaleSpriteSize(
    Uniform Curve Scale=1.5
    Uniform Curve Sprite Scale=DataInterface() {propertyValues {Curve {keys [
        {interpMode=RCIM_Cubic, tangentMode=RCTM_User, time=0, value=1}
        {arriveTangent=-1.6179, interpMode=RCIM_Cubic, leaveTangent=-1.6179, tangentMode=RCTM_User, time=1, value=0.5}
       ]}}}
   )
   &ScaleSpriteSizeBySpeed(
    Max Scale Factor=<1 12>
    Min Scale Factor=<0.6 1>
    Velocity Threshold:NiagaraFloat=&Lerp_Float(
     A=500
     B=5000
     Alpha=&FloatFromCurve(
      FloatCurve=DataInterface() {propertyValues {Curve {keys [
          {interpMode=RCIM_Cubic, tangentMode=RCTM_User, time=0, value=0}
          {arriveTangent=2.3931, interpMode=RCIM_Cubic, leaveTangent=2.3931, tangentMode=RCTM_User, time=0.8, value=0.6}
          {arriveTangent=0.6757, interpMode=RCIM_Cubic, leaveTangent=0.6757, tangentMode=RCTM_User, time=1, value=1}
         ]}}}
      CurveIndex=&RandomRangeFloat(Minimum=0, Maximum=1)
     )
    )
   )
  ]
  rendererClasses [&NiagaraSpriteRendererProperties]
  renderers [{rendererClass=&NiagaraSpriteRendererProperties}]
 }
 SpinBlur {
  emitterUpdateScript [
   &EmitterState(
    Inactive Response=::NewEnumerator1
    Life Cycle Mode=::NewEnumerator1
    Loop Behavior=::NewEnumerator1
    Loop Duration Mode=::NewEnumerator1
    UseLoopDelay=false
   )
   &SpawnBurst_Instantaneous(Use Spawn Probability=false, UseLoopCountLimit=false)
  ]
  particleSpawnScript [
   &InitializeParticle(
    Color=<rgba 1 1 1 0.5>
    Lifetime=5
    Sprite Size Mode=::NewEnumerator3
    Uniform Sprite Size=185
    UsePositionOffset=false
    Write Material Random=true
   )
   /SetVariables(Particles.SpriteFacing=<0 1 0>) {isSetParameters=true}
   &CameraOffset(Camera Offset Amount=4)
  ]
  particleUpdateScript [
   &ParticleState(Kill Particles When Lifetime Has Elapsed=false)
   &SpriteRotationRate(Rotation Rate=3000)
  ]
  rendererClasses [&NiagaraSpriteRendererProperties]
  renderers [{rendererClass=&NiagaraSpriteRendererProperties}]
 }
 Wind {
  emitterUpdateScript [
   &EmitterState(Inactive Response=::NewEnumerator1, Life Cycle Mode=::NewEnumerator1, Loop Duration=1, UseLoopDelay=false)
   &SpawnRate(
    Spawn Probability=&MakeCustomFloatFromBool(Bool=$User.bIsAirborne, True Float=1, False Float=0)
    SpawnRate=25
    Use Spawn Probability=true
   )
  ]
  particleSpawnScript [
   &InitializeParticle(
    Color=<rgba 0.8958 0.936 1 0.6>
    Lifetime Max=0.75
    Lifetime Min=0.4
    Lifetime Mode=::NewEnumerator1
    Mesh Scale=&Multiply_VectorByFloat(
     Vector=&RandomRangeVector(Minimum=<0.35 0.35 0.15>, Maximum=<0.6 0.6 0.2>)
     Float=&Lerp_Float(
      A=1
      B=1.5
      Alpha=&FloatFromCurve(
       FloatCurve=DataInterface() {propertyValues {Curve {keys [
           {interpMode=RCIM_Cubic, tangentMode=RCTM_User, time=0.6, value=0}
           {arriveTangent=5.6306, interpMode=RCIM_Cubic, leaveTangent=5.6306, tangentMode=RCTM_User, time=1, value=1}
          ]}}}
       CurveIndex=&RandomRangeFloat(Minimum=0, Maximum=1)
      )
     )
    )
    Mesh Scale Mode=::NewEnumerator4
    UsePositionOffset=false
    Write Material Random=true
   )
   &InitialMeshOrientation(
    Mesh Axis To Orient=<0 1 0>
    Mesh Orientation Mode=::NewEnumerator2
    Orientation Vector=<0 0 -1>
    Rotation=&RandomRangeVector(Minimum=<0 0 0>, Maximum=<0 0 1>)
    Use Rotation Vector=true
   )
  ]
  particleUpdateScript [
   &ParticleState()
   &ScaleColor(
    Scale Alpha=&FloatFromCurve(
     FloatCurve=DataInterface() {propertyValues {Curve {keys [
         {interpMode=RCIM_Cubic, tangentMode=RCTM_User, time=0, value=0}
         {interpMode=RCIM_Cubic, tangentMode=RCTM_User, time=0.5, value=1}
         {interpMode=RCIM_Cubic, leaveTangent=-2.3395, time=1, value=0}
        ]}}}
     Scale Curve=&RandomRangeFloat(Minimum=0.3, Maximum=1)
    )
    ScaleA=true
    ScaleRGB=false
   )
   &DynamicMaterialParameters(
    Index 0 Param 1=&FloatFromCurve(FloatCurve=DataInterface() {propertyValues {Curve {keys [<at 0.5 0 RCIM_Cubic>, {arriveTangent=4.311, interpMode=RCIM_Cubic, leaveTangent=16.4602, tangentMode=RCTM_User, time=1, value=1}]}}})
    Index 0 Param 2=&FloatFromCurve(
     FloatCurve=DataInterface() {propertyValues {Curve {keys [
         {arriveTangent=6.2888, interpMode=RCIM_Cubic, leaveTangent=1.845, tangentMode=RCTM_User, time=0, value=0}
         {arriveTangent=0.4139, interpMode=RCIM_Cubic, leaveTangent=-0.8461, tangentMode=RCTM_User, time=1, value=1}
        ]}}}
     Scale Curve=-0.5
    )
    Index 0 Param 4=&FloatFromCurve(
     FloatCurve=DataInterface() {propertyValues {Curve {keys [
         {arriveTangent=0.2805, interpMode=RCIM_Cubic, time=0, value=1}
         {arriveTangent=2.4494, interpMode=RCIM_Cubic, leaveTangent=2.4494, tangentMode=RCTM_User, time=1, value=2}
        ]}}}
     Scale Curve=1.5
    )
    Module.Param0WriteEnabled=true
    Module.Param1WriteEnabled=false
    Module.Param2WriteEnabled=false
    Module.Param3WriteEnabled=true
   )
   &UpdateMeshOrientation(Rotation Rate=&RandomRangeFloat(Minimum=2, Maximum=3), Rotation Vector=<0 0 1>)
   &ScaleMeshSize(Scale Factor=&VectorFromFloat(Value=&FloatFromCurve(
      FloatCurve=DataInterface() {propertyValues {Curve {keys [{interpMode=RCIM_Cubic, tangentMode=RCTM_User, time=0, value=0.65}, {interpMode=RCIM_Cubic, tangentMode=RCTM_User, time=1, value=1}]}}}
      Scale Curve=3
     )))
   &ScaleColor(ScaleA=true, ScaleRGB=false) {enabled=false, module=ScaleColor001}
  ]
  rendererClasses [&NiagaraMeshRendererProperties]
  renderers [{rendererClass=&NiagaraMeshRendererProperties, rendererIndex=0}]
 }
 sparkes {
  simTarget=GPUComputeSim
  emitterUpdateScript [
   &EmitterState()
   &SpawnRate(
    Spawn Probability=&MakeCustomFloatFromBool(Bool=$User.bIsAirborne, True Float=0, False Float=1)
    SpawnRate=150
    Use Spawn Probability=true
   )
  ]
  particleSpawnScript [
   /SetVariables(Particles.VelLength:NiagaraFloat=&VectorLength(Vector=$Engine.Owner.Velocity)) {isSetParameters=true}
   &InitializeParticle(
    Color=&LerpLinearColors(
     StartColor=<rgba 20 7.4236 0.1042 1>
     EndColor=<rgba 15 0.8397 0.1929 1>
     LerpFactor=&RandomRangeFloat(Minimum=&Ceil_Float(Float=0), Maximum=&Ceil_Float(Float=1))
    )
    Lifetime=&Lerp_Float(
     A=0.1
     B=0.8
     Alpha=&FloatFromCurve(
      FloatCurve=DataInterface() {propertyValues {Curve {keys [
          {interpMode=RCIM_Cubic, tangentMode=RCTM_User, time=0, value=0}
          {arriveTangent=1.4123, interpMode=RCIM_Cubic, leaveTangent=1.4123, tangentMode=RCTM_User, time=0.9, value=0.375}
          {arriveTangent=4.1333, interpMode=RCIM_Cubic, leaveTangent=4.1333, tangentMode=RCTM_User, time=1, value=1}
         ]}}}
      CurveIndex=&RandomRangeFloat(Minimum=0, Maximum=1)
     )
    )
    Position Offset=<0 0 -27>
    Sprite Size Mode=::NewEnumerator1
    Sprite UV Mode=::NewEnumerator3
    Uniform Sprite Size Max=7
    UsePositionOffset=true
    Write Material Random=true
   ) {module=InitializeParticle001}
   &FN_PhysicalMaterial(
    Default USE Particles Color=true
    Dirt Color:LinearColor=&LerpLinearColors(
     StartColor=<rgba 20 7.4236 0.1042 1>
     EndColor=<rgba 15 0.8397 0.1929 1>
     LerpFactor=&RandomRangeFloat(Minimum=&Ceil_Float(Float=0), Maximum=&Ceil_Float(Float=1))
    )
    Dirt Life Coef=0.1
    Dirt Visibility=0
    Grass Color:LinearColor=&LerpLinearColors(
     StartColor=<rgba 20 7.4236 0.1042 1>
     EndColor=<rgba 15 0.8397 0.1929 1>
     LerpFactor=&RandomRangeFloat(Minimum=&Ceil_Float(Float=0), Maximum=&Ceil_Float(Float=1))
    )
    Grass Life Coef=0.1
    Grass Visibility=0
    Metal Visibility=0
    MetalColor:LinearColor=&LerpLinearColors(
     StartColor=<rgba 20 7.4236 0.1042 1>
     EndColor=<rgba 15 0.8397 0.1929 1>
     LerpFactor=&RandomRangeFloat(Minimum=&Ceil_Float(Float=0), Maximum=&Ceil_Float(Float=1))
    )
    Physical Surface:NiagaraInt32=$User.PhysicalSurface
    Sand Color:LinearColor=&LerpLinearColors(
     StartColor=<rgba 20 7.4236 0.1042 1>
     EndColor=<rgba 15 0.8397 0.1929 1>
     LerpFactor=&RandomRangeFloat(Minimum=&Ceil_Float(Float=0), Maximum=&Ceil_Float(Float=1))
    )
    Sand Life Coef=0.1
    Sand Visibility=0
    Stone Visibility=0
    StoneColor:LinearColor=&LerpLinearColors(
     StartColor=<rgba 20 7.4236 0.1042 1>
     EndColor=<rgba 15 0.8397 0.1929 1>
     LerpFactor=&RandomRangeFloat(Minimum=&Ceil_Float(Float=0), Maximum=&Ceil_Float(Float=1))
    )
    Watar Visibility=0
    Water Color:LinearColor=<rgba 0.3949 0.698 1 1>
    Water Life Coef=0.5
    Wood Color:LinearColor=&LerpLinearColors(
     StartColor=<rgba 20 7.4236 0.1042 1>
     EndColor=<rgba 15 0.8397 0.1929 1>
     LerpFactor=&RandomRangeFloat(Minimum=&Ceil_Float(Float=0), Maximum=&Ceil_Float(Float=1))
    )
    Wood Visibility=0
   )
   &SphereLocation(
    MaskSpawnGroup=false
    Module.Override Randomness=false
    Non Uniform Scale=<2 1 1>
    Offset=HlslExpression() {hlslExpression="User.PhysicalSurface ==8? float3(0.0, 0.0, 15.0) : float3(0.0, 0.0, 15.0) /* Custom HLSL! */"}
    Sphere Radius=&Lerp_Float(
     A=3
     B=20
     Alpha=&MakeCustomFloatFromBool(
      Bool=&SetBoolByFloatComparison(A=$Particles.VelLength, B=300, Comparison Type=::NewEnumerator6)
      True Float=1
      False Float=0
     )
    )
    Surface Only=false
   )
   &SphereLocation(
    MaskSpawnGroup=false
    Module.Override Randomness=false
    Non Uniform Scale=<2 1 1>
    Offset=HlslExpression() {hlslExpression="User.PhysicalSurface ==8? float3(0.0, 0.0, 15.0) : float3(0.0, 0.0, -5.0) /* Custom HLSL! */"}
    Sphere Radius=2
    Surface Only=false
   ) {enabled=false, module=SphereLocation001}
   &AddVelocity(
    Scale Added Velocity=HlslExpression() {hlslExpression="float3(1.0, sign(rand(1.0)-0.5), 1.0) /* Custom HLSL! */"}
    Velocity=&Lerp_Vector(
     A=&RandomRangeVector(Minimum=<400 400 125>, Maximum=<-2000 1000 1200>)
     B=&RandomRangeVector(Minimum=<-50 600 100>, Maximum=<-200 1200 700>)
     Alpha=&MakeCustomFloatFromBool(
      Bool=&SetBoolByFloatComparison(A=$Particles.VelLength, B=300, Comparison Type=::NewEnumerator6)
      True Float=1
      False Float=0
     )
    )
   ) {module=AddVelocity001}
   &AddVelocity(
    Velocity=&RandomRangeVector(Minimum=<0 0 25>, Maximum=<0 0 150>)
    Velocity Speed Scale=&Lerp_Float(
     A=1
     B=0.5
     Alpha=&MakeCustomFloatFromBool(
      Bool=&SetBoolByFloatComparison(A=$Particles.VelLength, B=300, Comparison Type=::NewEnumerator6)
      True Float=1
      False Float=0
     )
    )
   ) {enabled=false}
   &ApplyInitialForces()
   /SetVariables(Particles.VelStopped:Vector3f=&Multiply_Vector(A=$Particles.Velocity, B=<0.5 0.5 0.5>)) {enabled=false, isSetParameters=true}
   /SetVariables() {enabled=false, isSetParameters=true}
  ]
  particleUpdateScript [
   &ParticleState()
   /Niagara/Modules/Update/Forces/CurlNoiseForce(
    Noise Frequency=0.1
    Noise Strength=&FloatFromCurve(
     FloatCurve=DataInterface() {propertyValues {Curve {keys [<at 0 0>, {arriveTangent=1, interpMode=RCIM_Cubic, leaveTangent=1, time=0.4106, value=0.6448}, <at 1 1>]}}}
     Scale Curve=300
    )
    Pan Noise=true
    Pan Noise Field=<0.5 -1 1>
   )
   &Drag(Drag=&RandomRangeFloat(Minimum=2, Maximum=4), Use Linear Drag=true, Use Rotational Drag=false)
   &ScaleColor(
    Scale Alpha=&Multiply_Float(
     A=&FloatFromCurve(FloatCurve=DataInterface() {propertyValues {Curve {keys [
          {arriveTangent=11.9433, interpMode=RCIM_Cubic, leaveTangent=11.9433, tangentMode=RCTM_User, time=0, value=0}
          <at 0.2 1 RCIM_Cubic>
          <at 1 0.7 RCIM_Cubic>
         ]}}})
     B=&Multiply_Float(A=1, B=&FNTimeOfDayScalar(Night Scale=0.5, Morning Scale=0.25, Evening Scale=0.25))
    )
    ScaleA=true
    ScaleRGB=false
   ) {module=ScaleColor001}
   &DynamicMaterialParameters(
    Index 0 Param 1=&FloatFromCurve(FloatCurve=DataInterface() {propertyValues {Curve {keys [<at 0 0>,<at 1 1>]}}})
    Index 0 Param 2=1
    Index 0 Param 3=1
    Index 0 Param 4=1
    Module.Param0WriteEnabled=true
    Module.Param1WriteEnabled=true
    Module.Param2WriteEnabled=true
    Module.Param3WriteEnabled=true
   )
   &ScaleSpriteSize(Scale Factor=&Vector2DFromFloat(Value=&FloatFromCurve(
      FloatCurve=DataInterface() {propertyValues {Curve {keys [
          {arriveTangent=-2.2491, interpMode=RCIM_Cubic, leaveTangent=-2.2491, tangentMode=RCTM_User, time=0, value=1.2}
          {arriveTangent=-0.2658, interpMode=RCIM_Cubic, leaveTangent=-0.2658, tangentMode=RCTM_User, time=0.18, value=0.99}
          <at 1 0.8 RCIM_Cubic>
         ]}}}
      Scale Curve=1.5
     )))
   &GravityForce(Gravity=&RandomRangeVector(Minimum=<0 0 -2000>, Maximum=<0 0 -1000>))
   &Collision(
    Advanced Aging Rate=1
    Age Colliding Particles=false
    Enable Rest State=false
    GPU Collision Type=::NewEnumerator2
    Module.EnableMaxCollisionCount=false
    Module.Randomize Collision Normal Vector=false
   )
   /SetVariables(Particles.Velocity=&Lerp_Vector(
     A=$Particles.Velocity
     B=$Particles.VelStopped
     Alpha=&MakeCustomFloatFromBool(
      Bool=&SetBoolByFloatComparison(A=$Particles.VelLength, B=300, Comparison Type=::NewEnumerator6)
      True Float=1
      False Float=0
     )
    )) {enabled=false, isSetParameters=true}
   &ScaleVelocity() {enabled=false}
   &SolveForcesAndVelocity(Clamp Velocity=false, Limit Acceleration=false)
   &ScaleSpriteSizeBySpeed(Max Scale Factor=&RandomRangeVector2D(Minimum=<0.4 0.6>, Maximum=<0.4 7>), Min Scale Factor=<0.4 0.5>)
   &CameraOffset(Camera Offset Amount=10)
   &ScaleColor(
    Scale RGB=&Multiply_VectorByFloat(
     Vector=&Lerp_Vector(
      A=<1 1 1>
      B=<1 1 1>
      Alpha=&MakeCustomFloatFromBool(
       Bool=&SetBoolByFloatComparison(A=$Particles.VelLength, B=300, Comparison Type=::NewEnumerator6)
       True Float=1
       False Float=0
      )
     )
     Float=&FloatFromCurve(FloatCurve=DataInterface() {propertyValues {Curve {keys [<at 0.1 8 RCIM_Cubic>,<at 0.7 1>]}}})
    )
    ScaleA=false
    ScaleRGB=true
   )
  ]
  rendererClasses [&NiagaraSpriteRendererProperties]
  renderers [{rendererClass=&NiagaraSpriteRendererProperties}]
 }
}
summary {
 systemName=NS_SawbladeGun_Blade_Sparks_loop
 emitters [
  {emitterName=sparkes, simTarget=GPUComputeSim, rendererClasses [&NiagaraSpriteRendererProperties]}
  {
   emitterName=NE_FN_PhysMatDebris
   rendererClasses [&NiagaraSpriteRendererProperties, &NiagaraMeshRendererProperties]
  }
  {emitterName=Smoke, rendererClasses [&NiagaraSpriteRendererProperties]}
  {emitterName=NE_FN_PhysMatHeatCore, rendererClasses [&NiagaraMeshRendererProperties]}
  {emitterName=SpinBlur, rendererClasses [&NiagaraSpriteRendererProperties]}
  {emitterName=NE_FN_PhysMatDirt001, rendererClasses [&NiagaraMeshRendererProperties]}
  {emitterName=DustTrail, rendererClasses [&NiagaraSpriteRendererProperties]}
  {emitterName=Sand, rendererClasses [&NiagaraSpriteRendererProperties]}
  {emitterName=RibbonTrail, rendererClasses [&NiagaraRibbonRendererProperties]}
  {emitterName=Sparks, rendererClasses [&NiagaraSpriteRendererProperties]}
  {emitterName=Wind, rendererClasses [&NiagaraMeshRendererProperties]}
 ]
 userVariables [
  User.bIsAirborne {type=NiagaraBool, defaultValue=true}
  User.bIsFrontEnd {type=NiagaraBool}
  User.bIsFrontEndPreview {type=NiagaraBool}
  User.ChargeUp {type=NiagaraFloat, defaultValue=0, description=""}
  User.InitRotZ {type=NiagaraFloat, defaultValue=0, description=""}
  User.PhysicalSurface {type=NiagaraInt32, defaultValue=8}
  User.RibbonIDOffset {type=NiagaraInt32}
  User.SkeletalMesh {
   type=NiagaraDataInterfaceSkeletalMesh
   defaultValue {
    struct=NiagaraExt_VariableValue_DataInterface
    value {dataInterface=&self:NiagaraDataInterfaceSkeletalMesh_0}
   }
  }
 ]
}
