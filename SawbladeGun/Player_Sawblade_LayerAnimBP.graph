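# Linked anim layer Blueprint for the Fortnite player skeleton that adds sawblade gun idle and charge animations to the character.
# Each update casts the pawn's current weapon to B_SawbladeGun_Athena and stores its charging state in IsCharging.
# ItemFullBodyOverride blends between additive idle and a WeaponCharge state machine (intro auto-transitioning to loop via inertialization), then applies the result to the incoming pose.
# The other Item Override layers contain no animation logic.
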AnimBlueprint /SawbladeGun/Player_Sawblade_LayerAnimBP
target_skeleton=&Fortnite_M_Avg_Player_Skeleton
graphs [
 AnimGraph {nodes [Output Pose() {layerGroup=None}]}
 EventGraph {nodes [
   Event BlueprintUpdateAnimation()
   > n1 = Cast To B_SawbladeGun_Athena!(execute, Object=Get CurrentWeapon!(self:Fort Pawn Object Reference=TryGetFortPawn()) -> (CurrentWeapon:Fort Weapon Object Reference)) -> (then, CastFailed, AsB Sawblade Gun Athena:B Sawblade Gun Athena Object Reference) {is=title K2Node_DynamicCast}
   > Set IsCharging(IsCharging:bool=GetIsChargingWeapon!(self:Fort Weapon Object Reference=@n1) -> (ReturnValue:bool)) -> (Output_Get:bool)
   TryGetPawnOwner()
  ]}
 ItemAimOffset {nodes [
   Input Pose!() -> (Pose, AimOffsetAlpha:double, Yaw:double, Pitch:double) {name=InPose_AimOffset}
   Input Pose!() -> (Pose) {name=InPose_PreAimOffsets}
   Output Pose() {layerGroup=Item Override}
  ]}
 ItemFallAdditive {nodes [
   Input Pose() {name=InPoseItemFallAdditive}
   Output Pose() {layerGroup=Item Override}
  ]}
 ItemFallLandAdditive {nodes [
   Input Pose() {name=InPoseItemFallLandAdditive}
   Output Pose() {layerGroup=Item Override}
  ]}
 ItemFallLevel2Override {nodes [
   Input Pose() {name=InPose_FallLevel2}
   Output Pose() {layerGroup=Item Override}
  ]}
 ItemFinalPoseOverride {nodes [
   Input Pose() {name=InFinalPose}
   Output Pose() {layerGroup=Item Override}
  ]}
 ItemFullBodyAimOffsets {nodes [
   Input Pose!() -> (Pose, FullBody_AimOffsetAlpha:double, Aim_Yaw:double, Aim_Pitch:double) {name=InPoseFullBodyAimOffsets}
   Input Pose!() -> (Pose) {name=InPosePreFullBodyAimOffsets}
   Output Pose() {layerGroup=Item Override}
  ]}
 ItemFullBodyOverride {nodes [
   n1 = "Sequence Player (additive) 'SawBlade_Idle'"!(bLoopAnimation:bool=True) -> (Pose) {sequence=/SharedWeaponAnims/Animation/SawbladeGun/SawBlade_Idle}
   AnimGraphNode_StateMachine_0 = WeaponCharge() {is=title AnimGraphNode_StateMachine}
   n2 = Get IsCharging() -> (IsCharging:bool)
   Output Pose(Result=Apply Additive!(
     Base=Input Pose() {name=InPoseFullBody}
     Additive=Blend Poses by bool(BlendPose_0=@AnimGraphNode_StateMachine_0, BlendPose_1=@n1, bActiveValue=@n2) {blendType=Linear}
     bAlphaBoolEnabled:bool=True
    ) -> (Pose) {alphaInputType=Bool, alphaBoolBlend {blendInTime=0.1, blendOption=Linear, blendOutTime=0.1}}) {layerGroup=Item Override}
  ]}
 WeaponCharge {
  entry=Charge_Intro
  path=ItemFullBodyOverride.AnimGraphNode_StateMachine_0.WeaponCharge
  ids=id
  nodes [
   AnimStateNode_0 = Charge_Intro() {is=title AnimStateNode}
   AnimStateNode_1 = Charge_Loop() {is=title AnimStateNode}
   AnimStateTransitionNode_0 = AnimStateTransitionNode() {bAutomaticRuleBasedOnSequencePlayerInState=true, logicType=TLT_Inertialization}
  ]
  links [{
    dst=Charge_Loop
    rule=ItemFullBodyOverride.AnimGraphNode_StateMachine_0.WeaponCharge.AnimStateTransitionNode_0.Transition
    src=Charge_Intro
    via=AnimStateTransitionNode_0
   }]
 }
 Charge_Intro {
  path=ItemFullBodyOverride.AnimGraphNode_StateMachine_0.WeaponCharge.AnimStateNode_0.Charge_Intro
  nodes [Output Animation Pose(Result="Sequence Player (additive) 'SawBlade_Fire_Charge_Intro'"() {bLoopAnimation=false, sequence=/SawbladeGun/Items/SawbladeGun/Animation/SawBlade_Fire_Charge_Intro})]
 }
 Charge_Loop {
  path=ItemFullBodyOverride.AnimGraphNode_StateMachine_0.WeaponCharge.AnimStateNode_1.Charge_Loop
  nodes [Output Animation Pose(Result="Sequence Player (additive) 'SawBlade_Fire_Charge_Loop'"() {sequence=/SawbladeGun/Items/SawbladeGun/Animation/SawBlade_Fire_Charge_Loop})]
 }
 Transition {path=ItemFullBodyOverride.AnimGraphNode_StateMachine_0.WeaponCharge.AnimStateTransitionNode_0.Transition, nodes [Result()]}
 ItemIdleOverride {nodes [
   Input Pose() {name=InPose_Idle}
   Output Pose() {layerGroup=Item Override}
  ]}
 ItemJetPackJumpAdditive {nodes [Output Pose() {layerGroup=Item Override}]}
 ItemJetPackStartAdditive {nodes [Output Pose() {layerGroup=Item Override}]}
 ItemJumpLoopAdditive {nodes [
   Input Pose() {name=InPoseItemJumpLoopAdditive}
   Output Pose() {layerGroup=Item Override}
  ]}
 ItemJumpOverride {nodes [
   Input Pose() {name=InPoseJump}
   Output Pose() {layerGroup=Item Override}
  ]}
 ItemJumpUpAdditive {nodes [
   Input Pose() {name=InPoseItemJumpUpAdditive}
   Output Pose() {layerGroup=Item Override}
  ]}
 ItemPreAimOffsets {nodes [
   Input Pose() {name=InPosePreAimOffsets}
   Output Pose() {layerGroup=Item Override}
  ]}
 ItemPreIK {nodes [
   Input Pose() {name=InPose_PreIK}
   Output Pose() {layerGroup=Item Override}
  ]}
 ItemRightHandIK {nodes [
   Input Pose() {name=InPosePreRightHandIK}
   Input Pose() {name=InPosePostRightHandIK}
   Output Pose() {layerGroup=Item Override}
  ]}
 ItemSkeletalControl {nodes [
   Input Pose()
   Output Pose() {layerGroup=Item Override}
  ]}
 ItemSwimJumpFallLoop {nodes [Output Pose() {layerGroup=Item Override}]}
 ItemSwimJumpSurfaceEnd {nodes [Output Pose() {layerGroup=Item Override}]}
 ItemSwimJumpSurfaceLoop {nodes [Output Pose() {layerGroup=Item Override}]}
 ItemUpperBody {nodes [
   Input Pose!() -> (Pose, InputParam:Fort Anim Input Adjusted Aim Structure) {name=InPoseUpperBody}
   Output Pose() {layerGroup=Item Override}
  ]}
 ItemZipLineStartAdditive {nodes [Output Pose() {layerGroup=Item Override}]}
]
variables [IsCharging {type=boolean}]
